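# Translucent unlit Niagara sprite material for the music effect on the JadeTowelGloss character, built around a chromatic-split texture.
# Two oppositely offset samples, spaced by the ChromaOffset dynamic parameter, are tinted, blended, desaturated and masked, then combined with particle color for emissive.
# Opacity multiplies particle alpha, the texture mask and a panning dissolve driven by a dynamic parameter, with selectable world-space or mesh UVs.
# Depth fade softens opacity except on forward and mobile paths.
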Material /BRCosmetics/Characters/Player/Female/Medium/Bodies/F_MED_JadeTowelGloss/FX/M_Character_JadeTowelGloss_Music
graph {
 properties {
  bAutomaticallySetUsageInEditor=false
  bUsedWithNiagaraSprites=true
  blendMode=BLEND_Translucent
  shadingModel=MSM_Unlit
  usedShadingModels=Unlit
 }
 nodes [
  n1 = ParticleColor()
  n2 = VectorParameter() {parameterName=Chroma1, defaultValue=<rgba 1.5 0 0.8249 1>}
  n3 = TextureCoordinate()
  n4 = Multiply(
   A=Constant2Vector() {g=0.05, r=0.05}
   B=DynamicParameter() {defaultValue=<rgba -2 0.1 1 1>, paramNames [Dissolve,ChromaOffset,Param3,Param4]}.ChromaOffset
  )
  n5 = TextureSampleParameter2D(UVs=Add(A=@n3, B=@n4)) {
   parameterName=BaseTexture
   texture=&/T_Character_JadeTowelGloss_Music
  }
  n6 = TextureSampleParameter2D(UVs=Subtract(A=@n3, B=@n4)) {
   parameterName=BaseTexture
   texture=&T_Character_JadeTowelGloss_Music
  }
  n7 = VectorParameter() {parameterName=Chroma2, defaultValue=<rgba 0 1.5 1.2326 1>}
  n8 = Desaturation(
   None=LinearInterpolate(A=Multiply(A=@n2.RGB, B=@n5.R), B=Multiply(A=@n6.R, B=@n7.RGB))
   Fraction=ScalarParameter() {defaultValue=-1.5, parameterName=Desaturation}
  )
  n9 = TextureSampleParameter2D() {
   parameterName=BaseTexture
   texture=&T_Character_JadeTowelGloss_Music
  }
  n10 = Power(
   Base=ChannelMaskParameter(None=@n9.RGBA) {group=Base, parameterName=BaseTexture_RGBselect, sortPriority=5}.R
   Exp=ScalarParameter() {defaultValue=1, parameterName=power}
  ) {constExponent=1}
  n11 = Saturate(None=Add(A=Multiply(A=@n8, B=@n10), B=@n1.RGB))
  MaterialFunctionCall(Emissive=LinearInterpolate(A=@n1.RGB, B=Multiply(A=@n11, B=@n1.RGB), Alpha=ScalarParameter() {defaultValue=1, parameterName=Use_Chroma})) {materialFunction=&MF_FX_EmissiveAdjustments}
  n12 = WorldPosition() {worldPositionShaderOffset=WPT_ExcludeAllShaderOffsets}
  n13 = Multiply(A=Abs(None=Constant3Vector() {constant=<rgba 15 20 1 0>})) {constB=-1}
  n14 = MaterialFunctionCall(UVs=ComponentMask(None=Divide(A=@n12.XYZ, B=@n13) {constB=15}) {g=true, r=true}, "Rotation Angle (0-1)"=Constant() {r=0.25}) {materialFunction=&CustomRotator}
  n15 = TextureCoordinate()
  n16 = TextureSampleParameter2D(UVs=LinearInterpolate(
    A=Panner(Coordinate=@n14.Rotated Values) {speedX=0.2}
    B=Panner(Coordinate=@n15) {speedX=0.2}
    Alpha=ScalarParameter() {parameterName=Use_ChangeDissolve}
   )) {
   parameterName=DissolveTex
   samplerType=SAMPLERTYPE_LinearColor
   texture=/Game/TheOutpost/Textures/AnimatedOutpost_Base046_rgb_XVector_a_NormalizedHierPos_UV_2
  }
  n17 = DynamicParameter() {defaultValue=<rgba -2 0.1 1 1>, paramNames [Dissolve,ChromaOffset,Param3,Param4]}
  n18 = Multiply(A=Multiply(A=@n1.A, B=@n10), B=Saturate(None=Subtract(A=@n16.R, B=@n17.Dissolve) {constB=-2}))
  n19 = DepthFade(Opacity=@n18, FadeDistance=ScalarParameter() {defaultValue=2, parameterName=DepthFade})
  ShadingPathSwitch(Default=@n19, Deferred=@n19, Forward=@n18, Mobile=@n18)
 ]
}
