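# Translucent, unlit, two-sided Niagara material for the JadeTowelGloss character's MusicLine effect, usable on ribbons, sprites and mesh particles.
# Emissive pulses a panning color texture, tinted by particle color.
# Opacity blends scrolling noise over a distorted gradient with an optional twinkling star layer, then applies edge, view-angle and depth fades.
# Panning noise drives per-axis vertex offset along normals, with dynamic parameters varying speed, pulse, star brightness and offset strength.
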Material /BRCosmetics/Characters/Player/Female/Medium/Bodies/F_MED_JadeTowelGloss/FX/M_Character_JadeTowelGloss_MusicLine
graph {
 properties {
  bAutomaticallySetUsageInEditor=false
  bUsedWithNiagaraMeshParticles=true
  bUsedWithNiagaraRibbons=true
  bUsedWithNiagaraSprites=true
  blendMode=BLEND_Translucent
  shadingModel=MSM_Unlit
  twoSided=true
  usedShadingModels=Unlit
 }
 nodes [
  n1 = Transform(None=Constant3Vector() {constant=<rgba 1 0 0 0>}) {transformSourceType=TRANSFORMSOURCE_Local}
  n2 = DynamicParameter() {parameterIndex=1, paramNames [Texture_Offset_1, Texture_Offset_2, WPO_Mult, CameraDepthFade]}
  n3 = Multiply(
   A=Add(A=TextureCoordinate(), B=AppendVector(A=@n2.Texture_Offset_1, B=@n2.Texture_Offset_1))
   B=AppendVector(A=ScalarParameter() {defaultValue=1, parameterName=WPOTextureU}, B=ScalarParameter() {parameterName=WPOTextureV})
  )
  n4 = DynamicParameter() {defaultValue=<rgba 1 0.5 100 1>, paramNames [LerpTexturespeed_Offset, Bright_Random, star_bright, Param4]}
  n5 = AppendVector(
   A=Multiply(A=ScalarParameter() {defaultValue=-1, parameterName=WPOTextureSpeedX}, B=@n4.LerpTexturespeed_Offset)
   B=ScalarParameter() {parameterName=WPOTextureSpeedY}
  )
  n6 = TextureSampleParameter2D(UVs=MaterialFunctionCall(UVs=Panner(Coordinate=@n3, Speed=@n5) {speedX=-2}, "Rotation Angle (0-1)"=Constant() {r=0.25}) {materialFunction=&CustomRotator}.Rotated Values) {
   parameterName=WPOTexture
   samplerType=SAMPLERTYPE_Normal
   texture=/Game/Characters/Player/Female/Medium/Bodies/F_MED_Bistro_Astronaut/FX/T_GaussianNoise_Normal
  }
  n7 = Multiply(A=VertexNormalWS(), B=LinearInterpolate(Alpha=@n6.R))
  n8 = DynamicParameter() {parameterIndex=1, paramNames [Texture_Offset_1, Texture_Offset_2, WPO_Mult, CameraDepthFade]}
  n9 = Multiply(A=Multiply(A=Multiply(A=@n1, B=@n7), B=ScalarParameter() {parameterName=X WPO_Intensity}) {constB=10}, B=@n8.WPO_Mult)
  n10 = Transform(None=Constant3Vector() {constant=<rgba 0 1 0 0>}) {transformSourceType=TRANSFORMSOURCE_Local}
  n11 = Multiply(A=@n8.WPO_Mult, B=Multiply(A=Multiply(A=@n10, B=@n7), B=ScalarParameter() {parameterName=Y WPO_Intensity}) {constB=10})
  n12 = Constant3Vector() {constant=<rgba 0 0 1 0>}
  n13 = Multiply(
   A=Multiply(A=Transform(None=@n12) {transformSourceType=TRANSFORMSOURCE_Local}, B=@n7)
   B=ScalarParameter() {defaultValue=1, parameterName=Z WPO_Intensity}
  ) {constB=10}
  LinearInterpolate(
   A=Constant()
   B=Add(A=Add(A=@n9, B=@n11), B=Multiply(A=@n8.WPO_Mult, B=@n13))
   Alpha=ScalarParameter() {defaultValue=1, parameterName=Use_WPO}
  )
  n14 = DynamicParameter() {defaultValue=<rgba 1 0.5 100 1>, paramNames [LerpTexturespeed_Offset, Bright_Random, star_bright, Param4]}
  n15 = Multiply(A=Time(), B=@n14.Bright_Random)
  n16 = LinearInterpolate(Alpha=Abs(None=Sine(None=@n15) {period=2})) {constA=0.7, constB=2}
  n17 = TextureSampleParameter2D(UVs=Panner(Coordinate=TextureCoordinate()) {speedX=-0.5}) {
   parameterName=ColorTexture
   texture=&/T_Character_JadeTowelGloss_Color
  }
  n18 = Desaturation(None=@n17.RGB, Fraction=ScalarParameter() {defaultValue=-5, parameterName=Desaturation})
  n19 = ParticleColor()
  MaterialFunctionCall(Emissive=Multiply(A=Multiply(A=@n16, B=@n18) {constA=10}, B=@n19.RGB)) {materialFunction=&MF_FX_EmissiveAdjustments}
  ScalarParameter() {defaultValue=100, parameterName=star_bright}
  n20 = TextureCoordinate()
  n21 = DynamicParameter() {defaultValue=<rgba 1 0.5 100 1>, paramNames [LerpTexturespeed_Offset, Bright_Random, star_bright, Param4]}
  n22 = AppendVector(
   A=Multiply(A=ScalarParameter() {defaultValue=-0.8, parameterName=LerpTexture_SpeedX}, B=@n21.LerpTexturespeed_Offset)
   B=ScalarParameter() {parameterName=LerpTexture_SpeedY}
  )
  n23 = Power(
   Base=TextureSampleParameter2D(UVs=Panner(Coordinate=@n20, Speed=@n22) {speedX=-0.8}) {parameterName=LerpTexture, samplerType=SAMPLERTYPE_Grayscale, texture=&LowResBlurredNoise}.R
   Exp=ScalarParameter() {defaultValue=2.5, parameterName=LerpTexture_Power}
  ) {constExponent=1.5}
  n24 = TextureCoordinate()
  n25 = AppendVector(A=ScalarParameter() {defaultValue=1, parameterName=BaseTextureU}, B=ScalarParameter() {defaultValue=5, parameterName=BaseTextureV})
  n26 = TextureCoordinate() {uTiling=2, vTiling=0}
  n27 = TextureSample(UVs=MaterialFunctionCall(UVs=Panner(Coordinate=@n26) {speedX=-0.5}, "Rotation Angle (0-1)"=Constant() {r=0.25}) {materialFunction=&CustomRotator}.Rotated Values) {texture=/Game/Global/Textures/Default/Masks_and_Gradients/T_VertGradient_02_sRGB}
  n28 = TextureSampleParameter2D(UVs=Add(A=Multiply(A=@n24, B=@n25), B=Multiply(A=@n27.R) {constB=0.2})) {parameterName=BaseTexture, texture=/Game/Global/Textures/Default/Masks_and_Gradients/gradient_stasis}
  n29 = DynamicParameter() {defaultValue=<rgba 1 0.5 100 1>, paramNames [LerpTexturespeed_Offset, Bright_Random, star_bright, Param4]}
  n30 = Multiply(A=ScalarParameter() {defaultValue=-0.15, parameterName=StarTexture_SpeedX}, B=@n29.LerpTexturespeed_Offset)
  n31 = ScalarParameter() {parameterName=StarTexture_SpeedY}
  n32 = TextureSample(UVs=Panner(Coordinate=TextureCoordinate() {uTiling=1.2, vTiling=0.15}, Speed=AppendVector(A=@n30, B=@n31)) {speedX=-0.8}) {
   samplerType=SAMPLERTYPE_Masks
   texture=/Game/Environments/Helios/Props/Quest/Quest_GoldenFleeceStatuette_A/Textures/T_Quest_GoldenFleeceStatuette_ShinyStars
  }
  n33 = TextureCoordinate() {uTiling=0.3, vTiling=0.3}
  n34 = TextureSample(UVs=Panner(Coordinate=Multiply(A=@n33) {constB=2}) {speedX=-0.4}) {samplerType=SAMPLERTYPE_LinearColor, texture=/Game/TheOutpost/Textures/AnimatedOutpost_Base046_rgb_XVector_a_NormalizedHierPos_UV_2}
  n35 = Time()
  n36 = Sine(None=Add(A=Power(Base=@n34.B), B=Multiply(A=@n35)))
  n37 = Add(A=TextureCoordinate() {uTiling=0.3, vTiling=0.3}, B=Constant2Vector() {g=0.15, r=0.15})
  n38 = TextureSample(UVs=Panner(Coordinate=Multiply(A=@n37) {constB=2}) {speedX=-0.3}) {samplerType=SAMPLERTYPE_LinearColor, texture=&AnimatedOutpost_Base046_rgb_XVector_a_NormalizedHierPos_UV_2}
  n39 = Time()
  n40 = Sine(None=Add(A=Power(Base=@n38.B), B=Multiply(A=@n39) {constB=0.3}))
  n41 = TextureCoordinate()
  n42 = ComponentMask(None=@n41) {g=true}
  n43 = Multiply(A=Multiply(A=ComponentMask(None=@n41) {g=true}, B=OneMinus(None=@n42))) {constB=2}
  n44 = Multiply(A=Multiply(A=@n32.R, B=Multiply(A=@n36, B=@n40)), B=Saturate(None=Power(Base=@n43)))
  n45 = DynamicParameter() {defaultValue=<rgba 1 0.5 100 1>, paramNames [LerpTexturespeed_Offset, Bright_Random, star_bright, Param4]}
  n46 = Multiply(
   A=@n19.A
   B=LinearInterpolate(A=Multiply(A=@n23, B=@n28.R), B=Multiply(A=@n44, B=@n45.star_bright) {constB=80}, Alpha=ScalarParameter() {parameterName=Use_star})
  )
  n47 = TextureCoordinate()
  n48 = Multiply(A=ComponentMask(None=@n47) {r=true}, B=OneMinus(None=ComponentMask(None=@n47) {r=true}))
  n49 = Saturate(None=Power(Base=Multiply(A=@n48) {constB=10}) {constExponent=3})
  n50 = Multiply(
   A=LinearInterpolate(A=@n46, B=Multiply(A=@n46, B=@n49), Alpha=ScalarParameter() {parameterName=Use_EdgeMask})
   B=MaterialFunctionCall("FadeSensitivity (optional)"=ScalarParameter() {defaultValue=2, parameterName=MeshFade}) {materialFunction=&FadeBasedOnViewAngle}.Result
  )
  n51 = DepthFade(Opacity=@n50, FadeDistance=ScalarParameter() {defaultValue=5, parameterName=DepthFade})
  ShadingPathSwitch(Default=@n51, Deferred=@n51, Forward=@n50, Mobile=@n50)
 ]
}
