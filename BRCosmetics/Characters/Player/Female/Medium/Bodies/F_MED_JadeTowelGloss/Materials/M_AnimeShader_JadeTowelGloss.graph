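# Anime-style cloth-shaded character material for the JadeTowelGloss female body, with an audio-reactive emissive glow.
# It chains the shared 2023 anime shader Pre, Calculations and Post functions through material attributes.
# A selectable FX-channel mask, an inverted Fresnel favoring camera-facing surfaces and a power-shaped audio amplitude multiply together.
# Tunable intensity and color parameters scale that product, which feeds the Post function's Emissive input.
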Material /BRCosmetics/Characters/Player/Female/Medium/Bodies/F_MED_JadeTowelGloss/Materials/M_AnimeShader_JadeTowelGloss
graph {
 properties {
  bAutomaticallySetUsageInEditor=false
  bUseMaterialAttributes=true
  bUsedWithMeshDeformer=true
  bUsedWithMorphTargets=true
  bUsedWithNiagaraMeshParticles=true
  bUsedWithSkeletalMesh=true
  bUsedWithStaticLighting=true
  pixelDepthOffsetMode=PDOM_Legacy
  shadingModel=MSM_Cloth
  usedShadingModels=Cloth
 }
 nodes [
  n1 = MaterialFunctionCall!(CustomUV) {materialFunction=/Game/Global/Materials/Character_Parent_2023/Anime_Parent_2023/MaterialFunctions/MFC_AnimeShader_2023_Pre}
  n2 = MaterialFunctionCall!(CustomLightVector, OverrideNormal_WS, STT=@n1.STT, Normal_Map=@n1.STT, DFL=@n1.STT, FX=@n1.STT) {materialFunction=/Game/Global/Materials/Character_Parent_2023/Anime_Parent_2023/MaterialFunctions/MFC_AnimeShader_2023_Calculations}
  n3 = Power(
   Base=ChannelMaskParameter(None=@n1.FX) {group=04_AudioReactivity, parameterName=ReactivChannel}.R
   Exp=ScalarParameter() {defaultValue=2, group=04_AudioReactivity, parameterName=MaskExponent, sliderMax=8}
  ) {constExponent=4}
  n4 = OneMinus(None=Fresnel() {baseReflectFraction=0, exponent=1})
  n5 = ScalarParameter() {defaultValue=2, group=04_AudioReactivity, parameterName=FresnelExponent, sliderMax=8}
  n6 = MaterialFunctionCall() {materialFunction=/Game/Packages/Fortress_SharedMaterials/Base_Material_Functions/BaseCharacterShaderFunctions/Audio/MF_AudioReactivity}
  n7 = Multiply(A=Multiply(A=Saturate(None=@n3), B=Power(Base=@n4, Exp=@n5) {constExponent=4}), B=Saturate(None=Power(Base=@n6.Amplitude) {constExponent=3}))
  n8 = ScalarParameter() {defaultValue=5, group=04_AudioReactivity, parameterName=EmissiveIntensity, sliderMax=10}
  MaterialFunctionCall!(
   Shading_Gradient=@n2.Shading_Gradient
   Threshold_Gradient=@n2.Shading_Gradient
   Specular_Gradient=@n2.Shading_Gradient
   Sheen_Mask=@n2.Shading_Gradient
   Thickness=@n2.Shading_Gradient
   DFL_Mask=@n2.Shading_Gradient
   FX=@n2.Shading_Gradient
   Metallic_Gradient=@n2.Shading_Gradient
   RimLight_Gradient=@n2.Shading_Gradient
   NormalBuffer=@n2.Shading_Gradient
   --------
   CINE_LightSources=@n2.Shading_Gradient
   CINE_LightColor=@n2.Shading_Gradient
   ---------
   Color_Lit=@n1.Color_Lit
   Color_Shaded=@n1.Color_Lit
   SRM=@n1.Color_Lit
   Emissive=Multiply(
    A=Multiply(A=@n7, B=@n8) {constB=5}
    B=VectorParameter() {group=04_AudioReactivity, parameterName=Emissive_Color, defaultValue=<rgba 0 1 0.7 1>}.RGB
   )
   DFL_Color=@n1.Color_Lit
   Specular_Color=@n1.Color_Lit
   RimLight_Color=@n1.Color_Lit
  ) {materialFunction=/Game/Global/Materials/Character_Parent_2023/Anime_Parent_2023/MaterialFunctions/MFC_AnimeShader_2023_Post}
  NamedRerouteUsage()
 ]
}
