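# Physics asset for the F_MED_JadeTowelGloss female player body cosmetic, with capsule bodies for its tie and front skirt bones plus the spine and thighs.
# The tie hangs from spine_04 through a free joint at its base, then runs as a five-link chain with limited swing and locked twist.
# Spine links and the two-segment center skirt joint rotate freely.
# The thighs, left-front skirt base and skirt chain have no constraint to the spine.
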PhysicsAsset /BRCosmetics/Characters/Player/Female/Medium/Bodies/F_MED_JadeTowelGloss/Meshes/F_MED_JadeTowelGloss_Physics
bodies {
 base_skirt_fr_l [{
   shapeType=Capsule
   length=5
   radius=5
   shapeName=base_skirt_fr_l_capsule
   center=<-0.4635 0.3628 3.1267>
   rotation=<rot 90 -175.2364 -158.3612>
  }]
 base_tie_01 [{shapeType=Capsule, length=0.505, radius=0.505, shapeName=base_tie_01_capsule, rotation=<rot 0 -90 0>}]
 dyn_skirt_fr_c_01 [{shapeType=Capsule, length=29, radius=5, shapeName=dyn_skirt_fr_c_01_capsule, center=<6.33 0 -3.5505>, rotation=<rot 0 0 90>}]
 dyn_skirt_fr_c_02 [{shapeType=Capsule, length=28.5, radius=5.5, shapeName=dyn_skirt_fr_c_02_capsule, center=<9.3394 0 -3.6534>, rotation=<rot 0 0 90>}]
 dyn_tie_01 [{shapeType=Capsule, length=0.505, radius=0.505, shapeName=dyn_tie_01_capsule, rotation=<rot 0 89.9709 -88.5708>}]
 dyn_tie_02 [{shapeType=Capsule, length=2.9489, radius=2, shapeName=dyn_tie_02_capsule, center=<0.2 0 1.531>, rotation=<rot 78.7004 180 180>}]
 dyn_tie_03 [{shapeType=Capsule, length=3.3001, radius=2, shapeName=dyn_tie_03_capsule, center=<0.1867 -0.0001 1.3226>, rotation=<rot 90 0 0>}]
 dyn_tie_04 [{shapeType=Capsule, length=3.4879, radius=2.0032, shapeName=dyn_tie_04_capsule, center=<1.076 0 1.2544>, rotation=<rot 90 0 0>}]
 dyn_tie_05 [{shapeType=Capsule, length=5.5, radius=3.5, shapeName=dyn_tie_05_capsule, center=<4.8829 0 2.5693>, rotation=<rot 87.0442 -180 180>}]
 spine_01 [{shapeType=Capsule, length=10, radius=10, shapeName=spine_01_capsule, center=<-5.5304 0 0>}]
 spine_02 [{shapeType=Capsule, length=10, radius=11.5, shapeName=spine_02_capsule, center=<0 -1.1202 0>}]
 spine_03 [{shapeType=Capsule, length=10, radius=11.5, shapeName=spine_03_capsule, center=<0 -0.7455 0>}]
 spine_04 [{shapeType=Capsule, length=10, radius=10, shapeName=spine_04_capsule, center=<0 1.9119 0>}]
 spine_05 [{shapeType=Capsule, length=10, radius=10, shapeName=spine_05_capsule, center=<0 1.2856 0>}]
 thigh_l [{shapeType=Capsule, length=24.5, radius=15, shapeName=thigh_l_capsule, center=<-26.4983 -0.3757 1.3313>, rotation=<rot 90 0 0>}]
 thigh_r [{shapeType=Capsule, length=24.5, radius=15, shapeName=thigh_l_capsule, center=<26.4983 0.3757 -1.3313>, rotation=<rot 90 0 180>}]
}
constraints [
 {bone1Name=dyn_tie_02, bone2Name=dyn_tie_01, swing1LimitDegrees=15, swing2LimitDegrees=40, twistLimitDegrees=45, twistMotion=Locked}
 {bone1Name=dyn_tie_03, bone2Name=dyn_tie_02, swing1LimitDegrees=25, swing2LimitDegrees=30, twistLimitDegrees=45, twistMotion=Locked}
 {bone1Name=dyn_tie_04, bone2Name=dyn_tie_03, swing1LimitDegrees=25, swing2LimitDegrees=30, twistLimitDegrees=45, twistMotion=Locked}
 {bone1Name=dyn_tie_05, bone2Name=dyn_tie_04, swing1LimitDegrees=25, swing2LimitDegrees=30, twistLimitDegrees=45, twistMotion=Locked}
 {
  bone1Name=spine_02
  bone2Name=spine_01
  swing1LimitDegrees=45
  swing1Motion=Free
  swing2LimitDegrees=45
  swing2Motion=Free
  twistLimitDegrees=45
  twistMotion=Free
 }
 {
  bone1Name=spine_03
  bone2Name=spine_02
  swing1LimitDegrees=45
  swing1Motion=Free
  swing2LimitDegrees=45
  swing2Motion=Free
  twistLimitDegrees=45
  twistMotion=Free
 }
 {
  bone1Name=spine_04
  bone2Name=spine_03
  swing1LimitDegrees=45
  swing1Motion=Free
  swing2LimitDegrees=45
  swing2Motion=Free
  twistLimitDegrees=45
  twistMotion=Free
 }
 {
  bone1Name=spine_05
  bone2Name=spine_04
  swing1LimitDegrees=45
  swing1Motion=Free
  swing2LimitDegrees=45
  swing2Motion=Free
  twistLimitDegrees=45
  twistMotion=Free
 }
 {
  bone1Name=dyn_skirt_fr_c_02
  bone2Name=dyn_skirt_fr_c_01
  swing1LimitDegrees=45
  swing1Motion=Free
  swing2LimitDegrees=45
  swing2Motion=Free
  twistLimitDegrees=45
  twistMotion=Free
 }
 {
  bone1Name=dyn_tie_01
  bone2Name=base_tie_01
  swing1LimitDegrees=0
  swing1Motion=Locked
  swing2LimitDegrees=15
  twistLimitDegrees=45
  twistMotion=Locked
 }
 {
  bone1Name=base_tie_01
  bone2Name=spine_04
  swing1LimitDegrees=45
  swing1Motion=Free
  swing2LimitDegrees=45
  swing2Motion=Free
  twistLimitDegrees=45
  twistMotion=Free
 }
]
