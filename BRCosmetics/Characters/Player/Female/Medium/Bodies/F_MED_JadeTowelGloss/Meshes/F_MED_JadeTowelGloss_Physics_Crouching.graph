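# Physics asset for the crouching variant of the female medium JadeTowelGloss player body, covering a dangling tie and the body around it.
# Five tie capsules form a chain hung from spine_04, with twist locked and swing limits that generally widen toward the tip.
# Pelvis, spine and thigh capsules approximate the body, with a double-capsule pelvis.
# Their joints are fully free, except the left thigh, which has locked twist and narrow swing limits.
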PhysicsAsset /BRCosmetics/Characters/Player/Female/Medium/Bodies/F_MED_JadeTowelGloss/Meshes/F_MED_JadeTowelGloss_Physics_Crouching
bodies {
 dyn_tie_01 [{shapeType=Capsule, length=2, radius=2, shapeName=dyn_tie_01_capsule, rotation=<rot 90 0 0>}]
 dyn_tie_02 [{shapeType=Capsule, length=4, radius=2, shapeName=dyn_tie_02_capsule, center=<0.1346 0 -0.2716>, rotation=<rot 78.7004 180 180>}]
 dyn_tie_03 [{shapeType=Capsule, length=4, radius=2, shapeName=dyn_tie_03_capsule, center=<-0.1867 -0.0001 0.0556>, rotation=<rot 90 0 0>}]
 dyn_tie_04 [{shapeType=Capsule, length=4, radius=2.5, shapeName=dyn_tie_04_capsule, center=<0 0 0.0099>, rotation=<rot 90 0 0>}]
 dyn_tie_05 [{shapeType=Capsule, length=5.5, radius=3.5, shapeName=dyn_tie_05_capsule, center=<4.1059 0 1.429>, rotation=<rot 90 0 0>}]
 pelvis [
  {shapeType=Capsule, length=17.1695, radius=24.3286, shapeName=pelvis_capsule, center=<-7.2166 7.7292 0>}
  {shapeType=Capsule, length=22, radius=8, shapeName=pelvis_capsule, center=<-8.163 32.8458 0>}
 ]
 spine_01 [{shapeType=Capsule, length=10, radius=10, shapeName=spine_01_capsule, center=<-5.5304 0 0>}]
 spine_02 [{shapeType=Capsule, length=10, radius=11.5, shapeName=spine_02_capsule}]
 spine_03 [{shapeType=Capsule, length=10, radius=11.5, shapeName=spine_03_capsule}]
 spine_04 [{shapeType=Capsule, length=10, radius=10, shapeName=spine_04_capsule, center=<0 1.9119 0>}]
 spine_05 [{shapeType=Capsule, length=10, radius=10, shapeName=spine_05_capsule, center=<0 1.2856 0>}]
 thigh_l [{shapeType=Capsule, length=24.5, radius=12.5, shapeName=thigh_l_capsule, center=<-32.5183 1.2807 1.3313>, rotation=<rot 90 0 0>}]
 thigh_r [{shapeType=Capsule, length=24.5, radius=12.5, shapeName=thigh_l_capsule, center=<32.5183 -1.2807 -1.3313>, rotation=<rot 90 0 180>}]
}
constraints [
 {bone1Name=dyn_tie_02, bone2Name=dyn_tie_01, swing1LimitDegrees=15, swing2LimitDegrees=30, twistLimitDegrees=45, twistMotion=Locked}
 {bone1Name=dyn_tie_03, bone2Name=dyn_tie_02, swing1LimitDegrees=15, swing2LimitDegrees=50, twistLimitDegrees=45, twistMotion=Locked}
 {bone1Name=dyn_tie_04, bone2Name=dyn_tie_03, swing1LimitDegrees=25, swing2LimitDegrees=50, twistLimitDegrees=45, twistMotion=Locked}
 {bone1Name=dyn_tie_05, bone2Name=dyn_tie_04, swing1LimitDegrees=25, swing2LimitDegrees=50, twistLimitDegrees=45, twistMotion=Locked}
 {
  bone1Name=spine_02
  bone2Name=spine_01
  swing1LimitDegrees=45
  swing1Motion=Free
  swing2LimitDegrees=45
  swing2Motion=Free
  twistLimitDegrees=45
  twistMotion=Free
 }
 {
  bone1Name=spine_03
  bone2Name=spine_02
  swing1LimitDegrees=45
  swing1Motion=Free
  swing2LimitDegrees=45
  swing2Motion=Free
  twistLimitDegrees=45
  twistMotion=Free
 }
 {
  bone1Name=spine_04
  bone2Name=spine_03
  swing1LimitDegrees=45
  swing1Motion=Free
  swing2LimitDegrees=45
  swing2Motion=Free
  twistLimitDegrees=45
  twistMotion=Free
 }
 {
  bone1Name=spine_05
  bone2Name=spine_04
  swing1LimitDegrees=45
  swing1Motion=Free
  swing2LimitDegrees=45
  swing2Motion=Free
  twistLimitDegrees=45
  twistMotion=Free
 }
 {
  bone1Name=spine_01
  bone2Name=pelvis
  swing1LimitDegrees=45
  swing1Motion=Free
  swing2LimitDegrees=45
  swing2Motion=Free
  twistLimitDegrees=45
  twistMotion=Free
 }
 {
  bone1Name=thigh_r
  bone2Name=pelvis
  swing1LimitDegrees=45
  swing1Motion=Free
  swing2LimitDegrees=45
  swing2Motion=Free
  twistLimitDegrees=45
  twistMotion=Free
 }
 {bone1Name=dyn_tie_01, bone2Name=spine_04, swing1LimitDegrees=20, swing2LimitDegrees=20, twistLimitDegrees=45, twistMotion=Locked}
 {bone1Name=thigh_l, bone2Name=pelvis, swing1LimitDegrees=20, swing2LimitDegrees=20, twistLimitDegrees=45, twistMotion=Locked}
]
