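# Animation Blueprint for the F_MED_JadeTowelGloss female body cosmetic, adding secondary motion to skirt, straps and sleeves.
# Copies the attached parent's pose into a Control Rig driven by gameplay state like crouch, melee, skydiving and emotes.
# Chained spring-based AnimDynamics nodes simulate cloth bones from a cached RigidBody pose.
# The Dynamics cache then gets additive bone offsets, plus tie constraints and a look-at weighted by Alpha_RB Disable.
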AnimBlueprint /BRCosmetics/Characters/Player/Female/Medium/Bodies/F_MED_JadeTowelGloss/Meshes/F_MED_JadeTowelGloss_AnimBP
target_skeleton=&/F_MED_JadeTowelGloss_Skeleton
graphs [
 AnimGraph {nodes [
   n1 = "Transform (Modify) Bone - Bone: base_tie_02"!(
    ComponentPose=Local To Component(LocalPose=Use cached pose 'Dynamics Cache'())
    bAlphaBoolEnabled:bool=True
    Translation="-1.409,-0.074,0.9"
    Rotation="0,0,0"
    Scale="1,1,1"
   ) -> (Pose) {
    alphaInputType=Bool
    rotationMode=BMM_Additive
    rotationSpace=BCS_BoneSpace
    translationMode=BMM_Additive
    translationSpace=BCS_BoneSpace
    alphaBoolBlend {blendInTime=0.2, blendOption=Linear, blendOutTime=0.5}
    alphaScaleBiasClamp {bInterpResult=true, interpSpeedDecreasing=5, interpSpeedIncreasing=5}
    boneToModify {boneName=base_tie_02}
    translation=<-1.409 -0.074 0.9>
   }
   n2 = "Transform (Modify) Bone - Bone: dyn_tie_03"!(ComponentPose=@n1, bAlphaBoolEnabled:bool=True, Translation="0,0.001,-3", Rotation="0,0,0", Scale="1,1,1") -> (Pose) {
    alphaInputType=Bool
    rotationMode=BMM_Additive
    rotationSpace=BCS_BoneSpace
    translationMode=BMM_Additive
    translationSpace=BCS_BoneSpace
    alphaBoolBlend {blendInTime=0.2, blendOption=Linear, blendOutTime=0.5}
    alphaScaleBiasClamp {bInterpResult=true, interpSpeedDecreasing=5, interpSpeedIncreasing=5}
    boneToModify {boneName=dyn_tie_03}
    translation {y=0.001, z=-3}
   }
   n3 = "Transform (Modify) Bone - Bone: base_strap_fr_l"(ComponentPose=@n2, Alpha=1, Translation="0,0,0", Rotation="-11.086,-0.000,-0.000", Scale="1,1,1") {
    rotationMode=BMM_Additive
    rotationSpace=BCS_BoneSpace
    translationMode=BMM_Additive
    translationSpace=BCS_BoneSpace
    alphaBoolBlend {blendInTime=0.2, blendOption=Linear, blendOutTime=0.5}
    alphaScaleBiasClamp {bInterpResult=true, interpSpeedDecreasing=5, interpSpeedIncreasing=5}
    boneToModify {boneName=base_strap_fr_l}
    rotation=<rot -11.086 0 0>
   }
   n4 = "Transform (Modify) Bone - Bone: sleeves_low_l"!(ComponentPose=@n3, bAlphaBoolEnabled:bool=True, Translation="-2,0,0", Rotation="0,0,0", Scale="1,1,1") -> (Pose) {
    alphaInputType=Bool
    rotationMode=BMM_Additive
    rotationSpace=BCS_BoneSpace
    translationMode=BMM_Additive
    translationSpace=BCS_BoneSpace
    alphaBoolBlend {blendInTime=0.2, blendOption=Linear, blendOutTime=0.2}
    alphaScaleBiasClamp {bInterpResult=true, interpSpeedDecreasing=5, interpSpeedIncreasing=5}
    boneToModify {boneName=sleeves_low_l}
    translation {x=-2}
   }
   n5 = "Transform (Modify) Bone - Bone: sleeves_low_l"(ComponentPose=@n4, bAlphaBoolEnabled=True, Translation="-4.85,0,0", Rotation="0,0,0", Scale="1,1,1") {
    alphaInputType=Bool
    rotationMode=BMM_Additive
    rotationSpace=BCS_BoneSpace
    translationMode=BMM_Additive
    translationSpace=BCS_BoneSpace
    alphaBoolBlend {blendInTime=0.2, blendOption=Linear, blendOutTime=0.2}
    alphaScaleBiasClamp {bInterpResult=true, interpSpeedDecreasing=5, interpSpeedIncreasing=5}
    boneToModify {boneName=sleeves_low_l}
    translation {x=-4.85}
   }
   n6 = "Transform (Modify) Bone - Bone: sleeves_low_r"!(ComponentPose=@n5, bAlphaBoolEnabled:bool=True, Translation="3.439,0.000,0.000", Rotation="0,0,0", Scale="1,1,1") -> (Pose) {
    alphaInputType=Bool
    rotationMode=BMM_Additive
    rotationSpace=BCS_BoneSpace
    translationMode=BMM_Additive
    translationSpace=BCS_BoneSpace
    alphaBoolBlend {blendInTime=0.2, blendOption=Linear, blendOutTime=0.2}
    alphaScaleBiasClamp {bInterpResult=true, interpSpeedDecreasing=5, interpSpeedIncreasing=5}
    boneToModify {boneName=sleeves_low_r}
    translation {x=3.439}
   }
   n7 = "Transform (Modify) Bone - Bone: sleeves_low_r"(ComponentPose=@n6, bAlphaBoolEnabled=True, Translation="2,0,0", Rotation="0,0,0", Scale="1,1,1") {
    alphaInputType=Bool
    rotationMode=BMM_Additive
    rotationSpace=BCS_BoneSpace
    translationMode=BMM_Additive
    translationSpace=BCS_BoneSpace
    alphaBoolBlend {blendInTime=0.2, blendOption=Linear, blendOutTime=0.2}
    alphaScaleBiasClamp {bInterpResult=true, interpSpeedDecreasing=5, interpSpeedIncreasing=5}
    boneToModify {boneName=sleeves_low_r}
    translation {x=2}
   }
   n8 = "Transform (Modify) Bone - Bone: base_strap_bk_l"(ComponentPose=@n7, Alpha=1, Translation="0,0,0", Rotation="38.607,0.070,0.067", Scale="1,1,1") {
    rotationMode=BMM_Additive
    rotationSpace=BCS_BoneSpace
    translationMode=BMM_Additive
    translationSpace=BCS_BoneSpace
    alphaBoolBlend {blendInTime=0.2, blendOption=Linear, blendOutTime=0.2}
    alphaScaleBiasClamp {bInterpResult=true, interpSpeedDecreasing=1, interpSpeedIncreasing=5}
    boneToModify {boneName=base_strap_bk_l}
    rotation=<rot 38.607 0.07 0.067>
   }
   n9 = "Transform (Modify) Bone - Bone: sleeves_low_r"!(ComponentPose=@n8, Alpha:float=1, Translation="2.85,0,0", Rotation="0,0,0", Scale="1,1,1") -> (Pose) {
    rotationMode=BMM_Additive
    rotationSpace=BCS_BoneSpace
    translationMode=BMM_Additive
    translationSpace=BCS_BoneSpace
    alphaBoolBlend {blendInTime=0.2, blendOption=Linear, blendOutTime=0.2}
    alphaScaleBiasClamp {bInterpResult=true, interpSpeedDecreasing=5, interpSpeedIncreasing=7}
    boneToModify {boneName=sleeves_low_r}
    translation {x=2.85}
   }
   n10 = Get Alpha_NotSkydiving() -> (Alpha_NotSkydiving:double)
   n11 = "Transform (Modify) Bone - Bone: MSTR_sleeve_r"(ComponentPose=@n9, Alpha=@n10, Translation="1.258,-0.174,0.000", Rotation="0.008,-4.703,0.004", Scale="1,1,1") {
    rotationMode=BMM_Additive
    rotationSpace=BCS_BoneSpace
    translationMode=BMM_Additive
    translationSpace=BCS_BoneSpace
    alphaBoolBlend {blendInTime=0.2, blendOption=Linear, blendOutTime=0.2}
    alphaScaleBiasClamp {bInterpResult=true, interpSpeedDecreasing=5, interpSpeedIncreasing=7}
    boneToModify {boneName=MSTR_sleeve_r}
    rotation=<rot 0.008 -4.703 0.004>
    translation=<1.258 -0.174>
   }
   n12 = "Transform (Modify) Bone - Bone: sleeves_low_l"!(ComponentPose=@n11, Alpha:float=@n10, Translation="-4.467,-0.000,0.000", Rotation="0,0,0", Scale="1,1,1") -> (Pose) {
    rotationMode=BMM_Additive
    rotationSpace=BCS_BoneSpace
    translationMode=BMM_Additive
    translationSpace=BCS_BoneSpace
    alphaBoolBlend {blendInTime=0.2, blendOption=Linear, blendOutTime=0.2}
    alphaScaleBiasClamp {bInterpResult=true, interpSpeedDecreasing=5, interpSpeedIncreasing=7}
    boneToModify {boneName=sleeves_low_l}
    translation {x=-4.467}
   }
   n13 = "Transform (Modify) Bone - Bone: base_skirt_sd_r"(ComponentPose=@n12, Alpha=1, Translation="0.743,0.000,2.291", Rotation="-48.443,-0.000,-0.000", Scale="1,1,1") {
    rotationMode=BMM_Additive
    rotationSpace=BCS_BoneSpace
    translationMode=BMM_Additive
    translationSpace=BCS_BoneSpace
    alphaBoolBlend {blendInTime=0.2, blendOption=Linear, blendOutTime=0.2}
    alphaScaleBiasClamp {bInterpResult=true, interpSpeedDecreasing=5, interpSpeedIncreasing=7}
    boneToModify {boneName=base_skirt_sd_r}
    rotation=<rot -48.443 0 0>
    translation {x=0.743, z=2.291}
   }
   n14 = "Transform (Modify) Bone - Bone: base_skirt_fr_r"(ComponentPose=@n13, Alpha=1, Translation="0,0,0", Rotation="-21.008,4.401,6.173", Scale="1,1,1") {
    rotationMode=BMM_Additive
    rotationSpace=BCS_BoneSpace
    translationMode=BMM_Additive
    translationSpace=BCS_BoneSpace
    alphaBoolBlend {blendInTime=0.2, blendOption=Linear, blendOutTime=0.2}
    alphaScaleBiasClamp {bInterpResult=true, interpSpeedDecreasing=5, interpSpeedIncreasing=7}
    boneToModify {boneName=base_skirt_fr_r}
    rotation=<rot -21.008 4.401 6.173>
   }
   n15 = "Transform (Modify) Bone - Bone: base_tie_01"!(ComponentPose=@n14, bAlphaBoolEnabled:bool=True, Translation="0,0,0.5", Rotation="3.5,0,0", Scale="1,1,1") -> (Pose) {
    alphaInputType=Bool
    rotationMode=BMM_Additive
    rotationSpace=BCS_BoneSpace
    translationMode=BMM_Additive
    translationSpace=BCS_BoneSpace
    alphaBoolBlend {blendInTime=0.2, blendOption=Linear, blendOutTime=0.2}
    alphaScaleBiasClamp {bInterpResult=true, interpSpeedDecreasing=5, interpSpeedIncreasing=7}
    boneToModify {boneName=base_tie_01}
    rotation {pitch=3.5}
    translation {z=0.5}
   }
   n16 = Get Alpha_RB Disable() -> (Alpha_RB Disable:double)
   n17 = "Constraint - dyn_tie_01 (P)"(ComponentPose=@n15, Alpha=@n16, ConstraintWeights_0=1) {
    alphaBoolBlend {blendOption=Linear}
    boneToModify {boneName=dyn_tie_01}
    constraintSetup [{targetBone {boneName=spine_05}}]
    constraintWeights [1]
   }
   n18 = "Constraint - base_tie_02 (P)"(ComponentPose=@n17, Alpha=@n16, ConstraintWeights_0=1) {
    alphaBoolBlend {blendOption=Linear}
    boneToModify {boneName=base_tie_02}
    constraintSetup [{targetBone {boneName=spine_04}}]
    constraintWeights [1]
   }
   n19 = "Constraint - dyn_tie_03 (P)"(ComponentPose=@n18, Alpha=@n16, ConstraintWeights_0=1) {
    alphaBoolBlend {blendOption=Linear}
    boneToModify {boneName=dyn_tie_03}
    constraintSetup [{targetBone {boneName=spine_03}}]
    constraintWeights [1]
   }
   n20 = "Constraint - dyn_tie_04 (P)"(ComponentPose=@n19, Alpha=@n16, ConstraintWeights_0=1) {
    alphaBoolBlend {blendOption=Linear}
    boneToModify {boneName=dyn_tie_04}
    constraintSetup [{targetBone {boneName=spine_02}}]
    constraintWeights [1]
   }
   n21 = Look At - Bone: dyn_tie_05(ComponentPose=@n20, Alpha=@n16) {
    alphaBoolBlend {blendOption=Linear}
    boneToModify {boneName=dyn_tie_05}
    lookAtLocation=<-0.7256 0.876 -1.0449>
    lookAtTarget {boneReference {boneName=dyn_skirt_fr_c_03}}
    lookAt_Axis {axis=<1 0>}
   }
   Output Pose(Result=Component To Local(ComponentPose=@n21)) {layerGroup=None}
   n22 = Control Rig!(
    Source=Copy Pose From Mesh() {bUseAttachedParent=true}
    Alpha
    PlayingEmote:double=Get Alpha_PlayingEmote() -> (Alpha_PlayingEmote:double)
    CrouchingMoving2D:double=Get Alpha_CrouchingMoving2D() -> (Alpha_CrouchingMoving2D:double)
    Crouching:double=Get Alpha_Crouching() -> (Alpha_Crouching:double)
    Melee_Bool:bool=Get bIsPlayingMeleeAnim()
    Crouching_Bool:bool=Get bIsCrouching()
    Skydiving:double=Get Alpha_NotSkydiving()
    DBNO:double=Get Alpha_DBNO() -> (Alpha_DBNO:double)
    FEIdle:double=Get FEIdle() -> (FEIdle:double)
   ) -> (Pose) {
    alphaBoolBlend {blendOption=Linear}
    controlRigAssetReference {blueprintRigClass=&/F_MED_JadeTowelGloss_CtrlRig.F_MED_JadeTowelGloss_CtrlRig_C}
   }
   Save cached pose 'CtrlRig Cache'(Pose=@n22) {cacheName=CtrlRig Cache}
   n23 = Anim Dynamics - Bone: dyn_skirt_fr_c_01(ComponentPose=Local To Component(LocalPose=Use cached pose 'RigidBodyCache'()), Alpha=1, ExternalForce="0,0,0") {
    angularDampingOverride=1
    angularSpringConstant=20000
    bAngularSpring=true
    bOverrideAngularDamping=true
    bUseGravityOverride=true
    gravityScale=0.35
    simulationSpace=RootRelative
    alphaBoolBlend {blendOption=Linear}
    alphaScaleBiasClamp {bInterpResult=true}
    boundBone {boneName=dyn_skirt_fr_c_01}
    componentAppliedLinearAccClamp=<3000 3000 3000>
    componentLinearAccScale=<0.65 0.65 0.65>
    componentLinearVelScale=<0.65 0.65 0.65>
    gravityOverride {z=-980}
    physicsBodyDefinitions [{
      boundBone {boneName=dyn_skirt_fr_c_01}
      boxExtents=<20 30 30>
      constraintSetup {angularLimitsMax=<5 0 10>, angularLimitsMin=<-5 -25 -10>, angularTarget=<1 0 0>}
      localJointOffset=<4 0 0>
     }]
    simSpaceSettings {simSpaceAngularAlpha=0.1}
   }
   n24 = Anim Dynamics - Bone: dyn_skirt_fr_c_02(ComponentPose=@n23, Alpha=1, ExternalForce="0,0,0") {
    angularDampingOverride=1
    angularSpringConstant=20000
    bAngularSpring=true
    bOverrideAngularDamping=true
    bUseGravityOverride=true
    gravityScale=0.35
    simulationSpace=RootRelative
    alphaBoolBlend {blendOption=Linear}
    alphaScaleBiasClamp {bInterpResult=true}
    boundBone {boneName=dyn_skirt_fr_c_02}
    componentAppliedLinearAccClamp=<3000 3000 3000>
    componentLinearAccScale=<0.65 0.65 0.65>
    componentLinearVelScale=<0.65 0.65 0.65>
    gravityOverride {z=-980}
    physicsBodyDefinitions [{
      boundBone {boneName=dyn_skirt_fr_c_02}
      boxExtents=<20 30 30>
      constraintSetup {angularLimitsMax=<5 0 10>, angularLimitsMin=<-5 -25 -10>, angularTarget=<1 0 0>}
      localJointOffset=<4 0 0>
     }]
    simSpaceSettings {simSpaceAngularAlpha=0.1}
   }
   n25 = Anim Dynamics - Bone: dyn_skirt_fr_c_03(ComponentPose=@n24, Alpha=1, ExternalForce="0,0,0") {
    angularDampingOverride=1
    angularSpringConstant=20000
    bAngularSpring=true
    bOverrideAngularDamping=true
    bUseGravityOverride=true
    gravityScale=0.35
    simulationSpace=RootRelative
    alphaBoolBlend {blendOption=Linear}
    alphaScaleBiasClamp {bInterpResult=true}
    boundBone {boneName=dyn_skirt_fr_c_03}
    componentAppliedLinearAccClamp=<3000 3000 3000>
    componentLinearAccScale=<0.65 0.65 0.65>
    componentLinearVelScale=<0.65 0.65 0.65>
    gravityOverride {z=-980}
    physicsBodyDefinitions [{
      boundBone {boneName=dyn_skirt_fr_c_03}
      boxExtents=<20 30 30>
      constraintSetup {angularLimitsMax=<5 0 10>, angularLimitsMin=<-5 -25 -10>, angularTarget=<1 0 0>}
      localJointOffset=<4 0 0>
     }]
    simSpaceSettings {simSpaceAngularAlpha=0.1}
   }
   n26 = Anim Dynamics - Bone: dyn_skirt_fr_l_01(ComponentPose=@n25, Alpha=1, ExternalForce="0,0,0") {
    angularDampingOverride=1
    angularSpringConstant=20000
    bAngularSpring=true
    bOverrideAngularDamping=true
    bUseGravityOverride=true
    gravityScale=0.35
    simulationSpace=RootRelative
    alphaBoolBlend {blendOption=Linear}
    alphaScaleBiasClamp {bInterpResult=true}
    boundBone {boneName=dyn_skirt_fr_l_01}
    componentAppliedLinearAccClamp=<3000 3000 3000>
    componentLinearAccScale=<0.65 0.65 0.65>
    componentLinearVelScale=<0.65 0.65 0.65>
    gravityOverride {z=-980}
    physicsBodyDefinitions [{
      boundBone {boneName=dyn_skirt_fr_l_01}
      boxExtents=<20 30 30>
      constraintSetup {angularLimitsMax=<5 0 5>, angularLimitsMin=<-5 -15 -5>, angularTarget=<1 0 0>}
      localJointOffset=<2 0 0>
     }]
    simSpaceSettings {simSpaceAngularAlpha=0.1}
   }
   n27 = Anim Dynamics - Bone: dyn_skirt_fr_l_02(ComponentPose=@n26, Alpha=1, ExternalForce="0,0,0") {
    angularDampingOverride=1
    angularSpringConstant=20000
    bAngularSpring=true
    bOverrideAngularDamping=true
    bUseGravityOverride=true
    gravityScale=0.35
    simulationSpace=RootRelative
    alphaBoolBlend {blendOption=Linear}
    alphaScaleBiasClamp {bInterpResult=true}
    boundBone {boneName=dyn_skirt_fr_l_02}
    componentAppliedLinearAccClamp=<3000 3000 3000>
    componentLinearAccScale=<0.65 0.65 0.65>
    componentLinearVelScale=<0.65 0.65 0.65>
    gravityOverride {z=-980}
    physicsBodyDefinitions [{
      boundBone {boneName=dyn_skirt_fr_l_02}
      boxExtents=<20 30 30>
      constraintSetup {angularLimitsMax=<5 0 5>, angularLimitsMin=<-5 -5 -5>, angularTarget=<1 0 0>}
      localJointOffset=<2 0 0>
     }]
    simSpaceSettings {simSpaceAngularAlpha=0.1}
   }
   n28 = Anim Dynamics - Bone: dyn_skirt_fr_l_03(ComponentPose=@n27, Alpha=1, ExternalForce="0,0,0") {
    angularDampingOverride=1
    angularSpringConstant=20000
    bAngularSpring=true
    bOverrideAngularDamping=true
    bUseGravityOverride=true
    gravityScale=0.35
    simulationSpace=RootRelative
    alphaBoolBlend {blendOption=Linear}
    alphaScaleBiasClamp {bInterpResult=true}
    boundBone {boneName=dyn_skirt_fr_l_03}
    componentAppliedLinearAccClamp=<3000 3000 3000>
    componentLinearAccScale=<0.65 0.65 0.65>
    componentLinearVelScale=<0.65 0.65 0.65>
    gravityOverride {z=-980}
    physicsBodyDefinitions [{
      boundBone {boneName=dyn_skirt_fr_l_03}
      boxExtents=<20 30 30>
      constraintSetup {angularLimitsMax=<5 0 5>, angularLimitsMin=<-5 -5 -15>, angularTarget=<1 0 0>}
      localJointOffset=<2 0 0>
     }]
    simSpaceSettings {simSpaceAngularAlpha=0.1}
   }
   n29 = Anim Dynamics - Bone: dyn_skirt_bk_c_01(ComponentPose=@n28, Alpha=1, ExternalForce="0,0,0") {
    angularDampingOverride=1
    angularSpringConstant=20000
    bAngularSpring=true
    bOverrideAngularDamping=true
    bUseGravityOverride=true
    gravityScale=0.35
    simulationSpace=RootRelative
    alphaBoolBlend {blendOption=Linear}
    alphaScaleBiasClamp {bInterpResult=true}
    boundBone {boneName=dyn_skirt_bk_c_01}
    componentAppliedLinearAccClamp=<3000 3000 3000>
    componentLinearAccScale=<0.65 0.65 0.65>
    componentLinearVelScale=<0.65 0.65 0.65>
    gravityOverride {z=-980}
    physicsBodyDefinitions [{
      boundBone {boneName=dyn_skirt_bk_c_01}
      boxExtents=<20 30 30>
      constraintSetup {angularLimitsMax=<5 0 10>, angularLimitsMin=<-5 -25 -10>, angularTarget=<1 0 0>}
      localJointOffset=<4 0 0>
     }]
    simSpaceSettings {simSpaceAngularAlpha=0.1}
   }
   n30 = Anim Dynamics - Bone: dyn_skirt_bk_c_02(ComponentPose=@n29, Alpha=1, ExternalForce="0,0,0") {
    angularDampingOverride=1
    angularSpringConstant=20000
    bAngularSpring=true
    bOverrideAngularDamping=true
    bUseGravityOverride=true
    gravityScale=0.35
    simulationSpace=RootRelative
    alphaBoolBlend {blendOption=Linear}
    alphaScaleBiasClamp {bInterpResult=true}
    boundBone {boneName=dyn_skirt_bk_c_02}
    componentAppliedLinearAccClamp=<3000 3000 3000>
    componentLinearAccScale=<0.65 0.65 0.65>
    componentLinearVelScale=<0.65 0.65 0.65>
    gravityOverride {z=-980}
    physicsBodyDefinitions [{
      boundBone {boneName=dyn_skirt_bk_c_02}
      boxExtents=<20 30 30>
      constraintSetup {angularLimitsMax=<5 0 10>, angularLimitsMin=<-5 -25 -10>, angularTarget=<1 0 0>}
      localJointOffset=<4 0 0>
     }]
    simSpaceSettings {simSpaceAngularAlpha=0.1}
   }
   n31 = Anim Dynamics - Bone: dyn_skirt_bk_c_03(ComponentPose=@n30, Alpha=1, ExternalForce="0,0,0") {
    angularDampingOverride=1
    angularSpringConstant=20000
    bAngularSpring=true
    bOverrideAngularDamping=true
    bUseGravityOverride=true
    gravityScale=0.35
    simulationSpace=RootRelative
    alphaBoolBlend {blendOption=Linear}
    alphaScaleBiasClamp {bInterpResult=true}
    boundBone {boneName=dyn_skirt_bk_c_03}
    componentAppliedLinearAccClamp=<3000 3000 3000>
    componentLinearAccScale=<0.65 0.65 0.65>
    componentLinearVelScale=<0.65 0.65 0.65>
    gravityOverride {z=-980}
    physicsBodyDefinitions [{
      boundBone {boneName=dyn_skirt_bk_c_03}
      boxExtents=<20 30 30>
      constraintSetup {angularLimitsMax=<5 0 10>, angularLimitsMin=<-5 -25 -10>, angularTarget=<1 0 0>}
      localJointOffset=<4 0 0>
     }]
    simSpaceSettings {simSpaceAngularAlpha=0.1}
   }
   n32 = Anim Dynamics - Bone: dyn_skirt_bk_r_01(ComponentPose=@n31, Alpha=1, ExternalForce="0,0,0") {
    angularDampingOverride=1
    angularSpringConstant=20000
    bAngularSpring=true
    bOverrideAngularDamping=true
    bUseGravityOverride=true
    gravityScale=0.35
    simulationSpace=RootRelative
    alphaBoolBlend {blendOption=Linear}
    alphaScaleBiasClamp {bInterpResult=true}
    boundBone {boneName=dyn_skirt_bk_r_01}
    componentAppliedLinearAccClamp=<3000 3000 3000>
    componentLinearAccScale=<0.65 0.65 0.65>
    componentLinearVelScale=<0.65 0.65 0.65>
    gravityOverride {z=-980}
    physicsBodyDefinitions [{
      boundBone {boneName=dyn_skirt_bk_r_01}
      boxExtents=<20 30 30>
      constraintSetup {angularLimitsMax=<5 0 10>, angularLimitsMin=<-5 -25 -10>, angularTarget=<1 0 0>}
      localJointOffset=<4 0 0>
     }]
    simSpaceSettings {simSpaceAngularAlpha=0.1}
   }
   n33 = Anim Dynamics - Bone: dyn_skirt_bk_r_02(ComponentPose=@n32, Alpha=1, ExternalForce="0,0,0") {
    angularDampingOverride=1
    angularSpringConstant=20000
    bAngularSpring=true
    bOverrideAngularDamping=true
    bUseGravityOverride=true
    gravityScale=0.35
    simulationSpace=RootRelative
    alphaBoolBlend {blendOption=Linear}
    alphaScaleBiasClamp {bInterpResult=true}
    boundBone {boneName=dyn_skirt_bk_r_02}
    componentAppliedLinearAccClamp=<3000 3000 3000>
    componentLinearAccScale=<0.65 0.65 0.65>
    componentLinearVelScale=<0.65 0.65 0.65>
    gravityOverride {z=-980}
    physicsBodyDefinitions [{
      boundBone {boneName=dyn_skirt_bk_r_02}
      boxExtents=<20 30 30>
      constraintSetup {angularLimitsMax=<5 0 10>, angularLimitsMin=<-5 -25 -10>, angularTarget=<1 0 0>}
      localJointOffset=<4 0 0>
     }]
    simSpaceSettings {simSpaceAngularAlpha=0.1}
   }
   n34 = Anim Dynamics - Bone: dyn_skirt_bk_r_03(ComponentPose=@n33, Alpha=1, ExternalForce="0,0,0") {
    angularDampingOverride=1
    angularSpringConstant=20000
    bAngularSpring=true
    bOverrideAngularDamping=true
    bUseGravityOverride=true
    gravityScale=0.35
    simulationSpace=RootRelative
    alphaBoolBlend {blendOption=Linear}
    alphaScaleBiasClamp {bInterpResult=true}
    boundBone {boneName=dyn_skirt_bk_r_03}
    componentAppliedLinearAccClamp=<3000 3000 3000>
    componentLinearAccScale=<0.65 0.65 0.65>
    componentLinearVelScale=<0.65 0.65 0.65>
    gravityOverride {z=-980}
    physicsBodyDefinitions [{
      boundBone {boneName=dyn_skirt_bk_r_03}
      boxExtents=<20 30 30>
      constraintSetup {angularLimitsMax=<5 0 10>, angularLimitsMin=<-5 -25 -10>, angularTarget=<1 0 0>}
      localJointOffset=<4 0 0>
     }]
    simSpaceSettings {simSpaceAngularAlpha=0.1}
   }
   n35 = Anim Dynamics - Bone: dyn_skirt_sd_l_01(ComponentPose=@n34, Alpha=1, ExternalForce="0,0,0") {
    angularDampingOverride=1
    angularSpringConstant=20000
    bAngularSpring=true
    bOverrideAngularDamping=true
    bUseGravityOverride=true
    gravityScale=0.35
    simulationSpace=RootRelative
    alphaBoolBlend {blendOption=Linear}
    alphaScaleBiasClamp {bInterpResult=true}
    boundBone {boneName=dyn_skirt_sd_l_01}
    componentAppliedLinearAccClamp=<3000 3000 3000>
    componentLinearAccScale=<0.65 0.65 0.65>
    componentLinearVelScale=<0.65 0.65 0.65>
    gravityOverride {z=-980}
    physicsBodyDefinitions [{
      boundBone {boneName=dyn_skirt_sd_l_01}
      boxExtents=<20 30 30>
      constraintSetup {angularLimitsMax=<5 0 10>, angularLimitsMin=<-5 -25 -10>, angularTarget=<1 0 0>}
      localJointOffset=<4 0 0>
     }]
    simSpaceSettings {simSpaceAngularAlpha=0.1}
   }
   n36 = Anim Dynamics - Bone: dyn_skirt_sd_l_02(ComponentPose=@n35, Alpha=1, ExternalForce="0,0,0") {
    angularDampingOverride=1
    angularSpringConstant=20000
    bAngularSpring=true
    bOverrideAngularDamping=true
    bUseGravityOverride=true
    gravityScale=0.35
    simulationSpace=RootRelative
    alphaBoolBlend {blendOption=Linear}
    alphaScaleBiasClamp {bInterpResult=true}
    boundBone {boneName=dyn_skirt_sd_l_02}
    componentAppliedLinearAccClamp=<3000 3000 3000>
    componentLinearAccScale=<0.65 0.65 0.65>
    componentLinearVelScale=<0.65 0.65 0.65>
    gravityOverride {z=-980}
    physicsBodyDefinitions [{
      boundBone {boneName=dyn_skirt_sd_l_02}
      boxExtents=<20 30 30>
      constraintSetup {angularLimitsMax=<5 0 10>, angularLimitsMin=<-5 -25 -10>, angularTarget=<1 0 0>}
      localJointOffset=<4 0 0>
     }]
    simSpaceSettings {simSpaceAngularAlpha=0.1}
   }
   n37 = Anim Dynamics - Bone: dyn_skirt_sd_l_03(ComponentPose=@n36, Alpha=1, ExternalForce="0,0,0") {
    angularDampingOverride=1
    angularSpringConstant=20000
    bAngularSpring=true
    bOverrideAngularDamping=true
    bUseGravityOverride=true
    gravityScale=0.35
    simulationSpace=RootRelative
    alphaBoolBlend {blendOption=Linear}
    alphaScaleBiasClamp {bInterpResult=true}
    boundBone {boneName=dyn_skirt_sd_l_03}
    componentAppliedLinearAccClamp=<3000 3000 3000>
    componentLinearAccScale=<0.65 0.65 0.65>
    componentLinearVelScale=<0.65 0.65 0.65>
    gravityOverride {z=-980}
    physicsBodyDefinitions [{
      boundBone {boneName=dyn_skirt_sd_l_03}
      boxExtents=<20 30 30>
      constraintSetup {angularLimitsMax=<5 0 10>, angularLimitsMin=<-5 -25 -10>, angularTarget=<1 0 0>}
      localJointOffset=<4 0 0>
     }]
    simSpaceSettings {simSpaceAngularAlpha=0.1}
   }
   n38 = Anim Dynamics - Bone: dyn_skirt_bk_l_01(ComponentPose=@n37, Alpha=1, ExternalForce="0,0,0") {
    angularDampingOverride=1
    angularSpringConstant=20000
    bAngularSpring=true
    bOverrideAngularDamping=true
    bUseGravityOverride=true
    gravityScale=0.35
    simulationSpace=RootRelative
    alphaBoolBlend {blendOption=Linear}
    alphaScaleBiasClamp {bInterpResult=true}
    boundBone {boneName=dyn_skirt_bk_l_01}
    componentAppliedLinearAccClamp=<3000 3000 3000>
    componentLinearAccScale=<0.65 0.65 0.65>
    componentLinearVelScale=<0.65 0.65 0.65>
    gravityOverride {z=-980}
    physicsBodyDefinitions [{
      boundBone {boneName=dyn_skirt_bk_l_01}
      boxExtents=<20 30 30>
      constraintSetup {angularLimitsMax=<5 0 10>, angularLimitsMin=<-5 -25 -10>, angularTarget=<1 0 0>}
      localJointOffset=<4 0 0>
     }]
    simSpaceSettings {simSpaceAngularAlpha=0.1}
   }
   n39 = Anim Dynamics - Bone: dyn_skirt_bk_l_02(ComponentPose=@n38, Alpha=1, ExternalForce="0,0,0") {
    angularDampingOverride=1
    angularSpringConstant=20000
    bAngularSpring=true
    bOverrideAngularDamping=true
    bUseGravityOverride=true
    gravityScale=0.35
    simulationSpace=RootRelative
    alphaBoolBlend {blendOption=Linear}
    alphaScaleBiasClamp {bInterpResult=true}
    boundBone {boneName=dyn_skirt_bk_l_02}
    componentAppliedLinearAccClamp=<3000 3000 3000>
    componentLinearAccScale=<0.65 0.65 0.65>
    componentLinearVelScale=<0.65 0.65 0.65>
    gravityOverride {z=-980}
    physicsBodyDefinitions [{
      boundBone {boneName=dyn_skirt_bk_l_02}
      boxExtents=<20 30 30>
      constraintSetup {angularLimitsMax=<5 0 10>, angularLimitsMin=<-5 -25 -10>, angularTarget=<1 0 0>}
      localJointOffset=<4 0 0>
     }]
    simSpaceSettings {simSpaceAngularAlpha=0.1}
   }
   n40 = Anim Dynamics - Bone: dyn_skirt_bk_l_03(ComponentPose=@n39, Alpha=1, ExternalForce="0,0,0") {
    angularDampingOverride=1
    angularSpringConstant=20000
    bAngularSpring=true
    bOverrideAngularDamping=true
    bUseGravityOverride=true
    gravityScale=0.35
    simulationSpace=RootRelative
    alphaBoolBlend {blendOption=Linear}
    alphaScaleBiasClamp {bInterpResult=true}
    boundBone {boneName=dyn_skirt_bk_l_03}
    componentAppliedLinearAccClamp=<3000 3000 3000>
    componentLinearAccScale=<0.65 0.65 0.65>
    componentLinearVelScale=<0.65 0.65 0.65>
    gravityOverride {z=-980}
    physicsBodyDefinitions [{
      boundBone {boneName=dyn_skirt_bk_l_03}
      boxExtents=<20 30 30>
      constraintSetup {angularLimitsMax=<5 0 10>, angularLimitsMin=<-5 -25 -10>, angularTarget=<1 0 0>}
      localJointOffset=<4 0 0>
     }]
    simSpaceSettings {simSpaceAngularAlpha=0.1}
   }
   n41 = Anim Dynamics - Bone: dyn_skirt_sd_r_01(ComponentPose=@n40, Alpha=1, ExternalForce="0,0,0") {
    angularDampingOverride=1
    angularSpringConstant=20000
    bAngularSpring=true
    bOverrideAngularDamping=true
    bUseGravityOverride=true
    gravityScale=0.35
    simulationSpace=RootRelative
    alphaBoolBlend {blendOption=Linear}
    alphaScaleBiasClamp {bInterpResult=true}
    boundBone {boneName=dyn_skirt_sd_r_01}
    componentAppliedLinearAccClamp=<3000 3000 3000>
    componentLinearAccScale=<0.65 0.65 0.65>
    componentLinearVelScale=<0.65 0.65 0.65>
    gravityOverride {z=-980}
    physicsBodyDefinitions [{
      boundBone {boneName=dyn_skirt_sd_r_01}
      boxExtents=<20 30 30>
      constraintSetup {angularLimitsMax=<5 0 10>, angularLimitsMin=<-5 -25 -10>, angularTarget=<1 0 0>}
      localJointOffset=<4 0 0>
     }]
    simSpaceSettings {simSpaceAngularAlpha=0.1}
   }
   n42 = Anim Dynamics - Bone: dyn_skirt_sd_r_02(ComponentPose=@n41, Alpha=1, ExternalForce="0,0,0") {
    angularDampingOverride=1
    angularSpringConstant=20000
    bAngularSpring=true
    bOverrideAngularDamping=true
    bUseGravityOverride=true
    gravityScale=0.35
    simulationSpace=RootRelative
    alphaBoolBlend {blendOption=Linear}
    alphaScaleBiasClamp {bInterpResult=true}
    boundBone {boneName=dyn_skirt_sd_r_02}
    componentAppliedLinearAccClamp=<3000 3000 3000>
    componentLinearAccScale=<0.65 0.65 0.65>
    componentLinearVelScale=<0.65 0.65 0.65>
    gravityOverride {z=-980}
    physicsBodyDefinitions [{
      boundBone {boneName=dyn_skirt_sd_r_02}
      boxExtents=<20 30 30>
      constraintSetup {angularLimitsMax=<5 0 10>, angularLimitsMin=<-5 -25 -10>, angularTarget=<1 0 0>}
      localJointOffset=<4 0 0>
     }]
    simSpaceSettings {simSpaceAngularAlpha=0.1}
   }
   n43 = Anim Dynamics - Bone: dyn_skirt_sd_r_03(ComponentPose=@n42, Alpha=1, ExternalForce="0,0,0") {
    angularDampingOverride=1
    angularSpringConstant=20000
    bAngularSpring=true
    bOverrideAngularDamping=true
    bUseGravityOverride=true
    gravityScale=0.35
    simulationSpace=RootRelative
    alphaBoolBlend {blendOption=Linear}
    alphaScaleBiasClamp {bInterpResult=true}
    boundBone {boneName=dyn_skirt_sd_r_03}
    componentAppliedLinearAccClamp=<3000 3000 3000>
    componentLinearAccScale=<0.65 0.65 0.65>
    componentLinearVelScale=<0.65 0.65 0.65>
    gravityOverride {z=-980}
    physicsBodyDefinitions [{
      boundBone {boneName=dyn_skirt_sd_r_03}
      boxExtents=<20 30 30>
      constraintSetup {angularLimitsMax=<5 0 10>, angularLimitsMin=<-5 -25 -10>, angularTarget=<1 0 0>}
      localJointOffset=<4 0 0>
     }]
    simSpaceSettings {simSpaceAngularAlpha=0.1}
   }
   n44 = Anim Dynamics - Bone: dyn_skirt_fr_r_01(ComponentPose=@n43, Alpha=1, ExternalForce="0,0,0") {
    angularDampingOverride=1
    angularSpringConstant=20000
    bAngularSpring=true
    bOverrideAngularDamping=true
    bUseGravityOverride=true
    gravityScale=0.35
    simulationSpace=RootRelative
    alphaBoolBlend {blendOption=Linear}
    alphaScaleBiasClamp {bInterpResult=true}
    boundBone {boneName=dyn_skirt_fr_r_01}
    componentAppliedLinearAccClamp=<3000 3000 3000>
    componentLinearAccScale=<0.65 0.65 0.65>
    componentLinearVelScale=<0.65 0.65 0.65>
    gravityOverride {z=-980}
    physicsBodyDefinitions [{
      boundBone {boneName=dyn_skirt_fr_r_01}
      boxExtents=<20 30 30>
      constraintSetup {angularLimitsMax=<5 0 5>, angularLimitsMin=<-5 -15 -5>, angularTarget=<1 0 0>}
      localJointOffset=<2 0 0>
     }]
    simSpaceSettings {simSpaceAngularAlpha=0.1}
   }
   n45 = Anim Dynamics - Bone: dyn_skirt_fr_r_02(ComponentPose=@n44, Alpha=1, ExternalForce="0,0,0") {
    angularDampingOverride=1
    angularSpringConstant=20000
    bAngularSpring=true
    bOverrideAngularDamping=true
    bUseGravityOverride=true
    gravityScale=0.35
    simulationSpace=RootRelative
    alphaBoolBlend {blendOption=Linear}
    alphaScaleBiasClamp {bInterpResult=true}
    boundBone {boneName=dyn_skirt_fr_r_02}
    componentAppliedLinearAccClamp=<3000 3000 3000>
    componentLinearAccScale=<0.65 0.65 0.65>
    componentLinearVelScale=<0.65 0.65 0.65>
    gravityOverride {z=-980}
    physicsBodyDefinitions [{
      boundBone {boneName=dyn_skirt_fr_r_02}
      boxExtents=<20 30 30>
      constraintSetup {angularLimitsMax=<5 0 5>, angularLimitsMin=<-5 -5 -5>, angularTarget=<1 0 0>}
      localJointOffset=<2 0 0>
     }]
    simSpaceSettings {simSpaceAngularAlpha=0.1}
   }
   n46 = Anim Dynamics - Bone: dyn_skirt_fr_r_03(ComponentPose=@n45, Alpha=1, ExternalForce="0,0,0") {
    angularDampingOverride=1
    angularSpringConstant=20000
    bAngularSpring=true
    bOverrideAngularDamping=true
    bUseGravityOverride=true
    gravityScale=0.35
    simulationSpace=RootRelative
    alphaBoolBlend {blendOption=Linear}
    alphaScaleBiasClamp {bInterpResult=true}
    boundBone {boneName=dyn_skirt_fr_r_03}
    componentAppliedLinearAccClamp=<3000 3000 3000>
    componentLinearAccScale=<0.65 0.65 0.65>
    componentLinearVelScale=<0.65 0.65 0.65>
    gravityOverride {z=-980}
    physicsBodyDefinitions [{
      boundBone {boneName=dyn_skirt_fr_r_03}
      boxExtents=<20 30 30>
      constraintSetup {angularLimitsMax=<5 -5 5>, angularLimitsMin=<-5 -15 -5>, angularTarget=<1 0 0>}
      localJointOffset=<2 0 0>
     }]
    simSpaceSettings {simSpaceAngularAlpha=0.1}
   }
   n47 = Anim Dynamics - Bone: dyn_strap_fr_l_01(ComponentPose=@n46, Alpha=1, ExternalForce="0,0,0") {
    angularDampingOverride=1
    angularSpringConstant=20000
    bAngularSpring=true
    bOverrideAngularDamping=true
    bUseGravityOverride=true
    gravityScale=0.35
    simulationSpace=RootRelative
    alphaBoolBlend {blendOption=Linear}
    boundBone {boneName=dyn_strap_fr_l_01}
    componentAppliedLinearAccClamp=<3000 3000 3000>
    componentLinearAccScale=<0.7 0.7 0.7>
    componentLinearVelScale=<1 1 1>
    gravityOverride {z=-980}
    physicsBodyDefinitions [{
      boundBone {boneName=dyn_strap_fr_l_01}
      boxExtents=<20 30 30>
      constraintSetup {angularLimitsMax=<5 0 0>, angularLimitsMin=<-5 -5 -5>, angularTarget=<1 0 0>}
      localJointOffset=<4 0 0>
     }]
   }
   n48 = Anim Dynamics - Bone: dyn_strap_fr_l_02(ComponentPose=@n47, Alpha=1, ExternalForce="0,0,0") {
    angularDampingOverride=1
    angularSpringConstant=20000
    bAngularSpring=true
    bOverrideAngularDamping=true
    bUseGravityOverride=true
    gravityScale=0.35
    simulationSpace=RootRelative
    alphaBoolBlend {blendOption=Linear}
    boundBone {boneName=dyn_strap_fr_l_02}
    componentAppliedLinearAccClamp=<3000 3000 3000>
    componentLinearAccScale=<0.7 0.7 0.7>
    componentLinearVelScale=<1 1 1>
    gravityOverride {z=-980}
    physicsBodyDefinitions [{
      boundBone {boneName=dyn_strap_fr_l_02}
      boxExtents=<20 30 30>
      constraintSetup {angularLimitsMax=<5 0 0>, angularLimitsMin=<-25 -5 -25>, angularTarget=<1 0 0>}
      localJointOffset=<4 0 0>
     }]
   }
   n49 = Anim Dynamics - Bone: dyn_strap_fr_l_03(ComponentPose=@n48, Alpha=1, ExternalForce="0,0,0") {
    angularDampingOverride=1
    angularSpringConstant=20000
    bAngularSpring=true
    bOverrideAngularDamping=true
    bUseGravityOverride=true
    gravityScale=0.35
    simulationSpace=RootRelative
    alphaBoolBlend {blendOption=Linear}
    boundBone {boneName=dyn_strap_fr_l_03}
    componentAppliedLinearAccClamp=<3000 3000 3000>
    componentLinearAccScale=<0.7 0.7 0.7>
    componentLinearVelScale=<1 1 1>
    gravityOverride {z=-980}
    physicsBodyDefinitions [{
      boundBone {boneName=dyn_strap_fr_l_03}
      boxExtents=<20 30 30>
      constraintSetup {angularLimitsMax=<0 0 0>, angularLimitsMin=<0 -5 -25>, angularTarget=<1 0 0>}
      localJointOffset=<4 0 0>
     }]
   }
   n50 = Anim Dynamics - Bone: dyn_strap_fr_l_04(ComponentPose=@n49, Alpha=1, ExternalForce="0,0,0") {
    angularDampingOverride=1
    angularSpringConstant=20000
    bAngularSpring=true
    bOverrideAngularDamping=true
    bUseGravityOverride=true
    gravityScale=0.35
    simulationSpace=RootRelative
    alphaBoolBlend {blendOption=Linear}
    boundBone {boneName=dyn_strap_fr_l_04}
    componentAppliedLinearAccClamp=<3000 3000 3000>
    componentLinearAccScale=<0.7 0.7 0.7>
    componentLinearVelScale=<1 1 1>
    gravityOverride {z=-980}
    physicsBodyDefinitions [{
      boundBone {boneName=dyn_strap_fr_l_04}
      boxExtents=<20 30 30>
      constraintSetup {angularLimitsMax=<0 0 5>, angularLimitsMin=<0 0 -10>, angularTarget=<1 0 0>}
      localJointOffset=<4 0 0>
     }]
   }
   n51 = Anim Dynamics - Bone: dyn_strap_bk_l_01(ComponentPose=@n50, Alpha=1, ExternalForce="0,0,0") {
    angularDampingOverride=1
    angularSpringConstant=20000
    bAngularSpring=true
    bOverrideAngularDamping=true
    bUseGravityOverride=true
    gravityScale=0.35
    simulationSpace=RootRelative
    alphaBoolBlend {blendOption=Linear}
    boundBone {boneName=dyn_strap_bk_l_01}
    componentAppliedLinearAccClamp=<3000 3000 3000>
    componentLinearAccScale=<0.7 0.7 0.7>
    componentLinearVelScale=<1 1 1>
    gravityOverride {z=-980}
    physicsBodyDefinitions [{
      boundBone {boneName=dyn_strap_bk_l_01}
      boxExtents=<20 30 30>
      constraintSetup {angularLimitsMax=<0 0 2>, angularLimitsMin=<-2 0 0>, angularTarget=<1 0 0>}
      localJointOffset=<4 0 0>
     }]
   }
   n52 = Anim Dynamics - Bone: dyn_strap_bk_l_02(ComponentPose=@n51, Alpha=1, ExternalForce="0,0,0") {
    angularDampingOverride=1
    angularSpringConstant=20000
    bAngularSpring=true
    bOverrideAngularDamping=true
    bUseGravityOverride=true
    gravityScale=0.35
    simulationSpace=RootRelative
    alphaBoolBlend {blendOption=Linear}
    boundBone {boneName=dyn_strap_bk_l_02}
    componentAppliedLinearAccClamp=<3000 3000 3000>
    componentLinearAccScale=<0.7 0.7 0.7>
    componentLinearVelScale=<1 1 1>
    gravityOverride {z=-980}
    physicsBodyDefinitions [{
      boundBone {boneName=dyn_strap_bk_l_02}
      boxExtents=<20 30 30>
      constraintSetup {angularLimitsMax=<2 0 20>, angularLimitsMin=<-20 0 0>, angularTarget=<1 0 0>}
      localJointOffset=<4 0 0>
     }]
   }
   n53 = Anim Dynamics - Bone: dyn_strap_bk_l_03(ComponentPose=@n52, Alpha=1, ExternalForce="0,0,0") {
    angularDampingOverride=1
    angularSpringConstant=20000
    bAngularSpring=true
    bOverrideAngularDamping=true
    bUseGravityOverride=true
    gravityScale=0.35
    simulationSpace=RootRelative
    alphaBoolBlend {blendOption=Linear}
    boundBone {boneName=dyn_strap_bk_l_03}
    componentAppliedLinearAccClamp=<3000 3000 3000>
    componentLinearAccScale=<0.7 0.7 0.7>
    componentLinearVelScale=<1 1 1>
    gravityOverride {z=-980}
    physicsBodyDefinitions [{
      boundBone {boneName=dyn_strap_bk_l_03}
      boxExtents=<20 30 30>
      constraintSetup {angularLimitsMax=<2 0 15>, angularLimitsMin=<-10 0 0>, angularTarget=<1 0 0>}
      localJointOffset=<4 0 0>
     }]
   }
   n54 = Anim Dynamics - Bone: dyn_strap_bk_l_04(ComponentPose=@n53, Alpha=1, ExternalForce="0,0,0") {
    angularDampingOverride=1
    angularSpringConstant=20000
    bAngularSpring=true
    bOverrideAngularDamping=true
    bUseGravityOverride=true
    gravityScale=0.35
    simulationSpace=RootRelative
    alphaBoolBlend {blendOption=Linear}
    boundBone {boneName=dyn_strap_bk_l_04}
    componentAppliedLinearAccClamp=<3000 3000 3000>
    componentLinearAccScale=<0.7 0.7 0.7>
    componentLinearVelScale=<1 1 1>
    gravityOverride {z=-980}
    physicsBodyDefinitions [{
      boundBone {boneName=dyn_strap_bk_l_04}
      boxExtents=<20 30 30>
      constraintSetup {angularLimitsMax=<0 0 0>, angularLimitsMin=<-20 0 -10>, angularTarget=<1 0 0>}
      localJointOffset=<4 0 0>
     }]
   }
   n55 = "Constraint - dyn_strap_bk_l_02 (R)"(ComponentPose=@n54, Alpha=0.2, ConstraintWeights_0=1) {
    alphaBoolBlend {blendOption=Linear}
    boneToModify {boneName=dyn_strap_bk_l_02}
    constraintSetup [{transformType=Rotation, perAxis {bY=false, bZ=false}, targetBone {boneName=dyn_skirt_bk_l_03}}]
    constraintWeights [1]
   }
   n56 = Anim Dynamics - Bone: dyn_sleeves_fr_l(ComponentPose=@n55, Alpha=1, ExternalForce="0,0,0") {
    angularDampingOverride=1
    angularSpringConstant=20000
    bAngularSpring=true
    bOverrideAngularDamping=true
    bUseGravityOverride=true
    gravityScale=0.35
    simulationSpace=RootRelative
    alphaBoolBlend {blendOption=Linear}
    boundBone {boneName=dyn_sleeves_fr_l}
    componentAppliedLinearAccClamp=<3000 3000 3000>
    componentLinearAccScale=<1 1 1>
    componentLinearVelScale=<1 1 1>
    gravityOverride {z=-980}
    physicsBodyDefinitions [{
      boundBone {boneName=dyn_sleeves_fr_l}
      boxExtents=<20 30 30>
      constraintSetup {angularLimitsMax=<15 15 15>, angularLimitsMin=<-15 -15 0>, angularTarget=<1 0 0>}
      localJointOffset=<2 0 0>
     }]
   }
   n57 = Anim Dynamics - Bone: dyn_sleeves_bk_l(ComponentPose=@n56, Alpha=1, ExternalForce="0,0,0") {
    angularDampingOverride=1
    angularSpringConstant=20000
    bAngularSpring=true
    bOverrideAngularDamping=true
    bUseGravityOverride=true
    gravityScale=0.35
    simulationSpace=RootRelative
    alphaBoolBlend {blendOption=Linear}
    boundBone {boneName=dyn_sleeves_bk_l}
    componentAppliedLinearAccClamp=<3000 3000 3000>
    componentLinearAccScale=<1 1 1>
    componentLinearVelScale=<1 1 1>
    gravityOverride {z=-980}
    physicsBodyDefinitions [{
      boundBone {boneName=dyn_sleeves_bk_l}
      boxExtents=<20 30 30>
      constraintSetup {angularLimitsMax=<15 15 15>, angularLimitsMin=<-15 -15 -15>, angularTarget=<1 0 0>}
      localJointOffset=<2 0 0>
     }]
   }
   n58 = Anim Dynamics - Bone: dyn_sleeves_bk_r(ComponentPose=@n57, Alpha=1, ExternalForce="0,0,0") {
    angularDampingOverride=1
    angularSpringConstant=20000
    bAngularSpring=true
    bOverrideAngularDamping=true
    bUseGravityOverride=true
    gravityScale=0.35
    simulationSpace=RootRelative
    alphaBoolBlend {blendOption=Linear}
    boundBone {boneName=dyn_sleeves_bk_r}
    componentAppliedLinearAccClamp=<3000 3000 3000>
    componentLinearAccScale=<1 1 1>
    componentLinearVelScale=<1 1 1>
    gravityOverride {z=-980}
    physicsBodyDefinitions [{
      boundBone {boneName=dyn_sleeves_bk_r}
      boxExtents=<20 30 30>
      constraintSetup {angularLimitsMax=<15 15 15>, angularLimitsMin=<-15 -15 -15>, angularTarget=<1 0 0>}
      localJointOffset=<2 0 0>
     }]
   }
   n59 = Anim Dynamics - Bone: dyn_sleeves_fr_r(ComponentPose=@n58, Alpha=1, ExternalForce="0,0,0") {
    angularDampingOverride=1
    angularSpringConstant=20000
    bAngularSpring=true
    bOverrideAngularDamping=true
    bUseGravityOverride=true
    gravityScale=0.35
    simulationSpace=RootRelative
    alphaBoolBlend {blendOption=Linear}
    boundBone {boneName=dyn_sleeves_fr_r}
    componentAppliedLinearAccClamp=<3000 3000 3000>
    componentLinearAccScale=<1 1 1>
    componentLinearVelScale=<1 1 1>
    gravityOverride {z=-980}
    physicsBodyDefinitions [{
      boundBone {boneName=dyn_sleeves_fr_r}
      boxExtents=<20 30 30>
      constraintSetup {angularLimitsMax=<15 15 15>, angularLimitsMin=<-15 -15 -15>, angularTarget=<1 0 0>}
      localJointOffset=<2 0 0>
     }]
   }
   n60 = Anim Dynamics - Bone: MSTR_sleeve_l(ComponentPose=@n59, Alpha=Get Alpha_NotSkydiving(), ExternalForce="0,0,0") {
    angularDampingOverride=1
    angularSpringConstant=20000
    bAngularSpring=true
    bOverrideAngularDamping=true
    bUseGravityOverride=true
    gravityScale=0.35
    simulationSpace=RootRelative
    alphaBoolBlend {blendOption=Linear}
    alphaScaleBiasClamp {bInterpResult=true, interpSpeedIncreasing=0.3}
    boundBone {boneName=MSTR_sleeve_l}
    componentAppliedLinearAccClamp=<3000 3000 3000>
    componentLinearAccScale=<1 1 1>
    componentLinearVelScale=<1 1 1>
    gravityOverride {z=-980}
    physicsBodyDefinitions [{
      boundBone {boneName=MSTR_sleeve_l}
      boxExtents=<20 30 30>
      constraintSetup {angularLimitsMax=<0 0 0>, angularLimitsMin=<0 0 -7>, angularTarget=<1 0 0>}
      localJointOffset=<3.5 0 0>
     }]
   }
   Save cached pose 'Dynamics Cache'(Pose=Component To Local(ComponentPose=@n60)) {cacheName=Dynamics Cache}
   n61 = "Transform (Modify) Bone - Bone: base_tie_01"!(
    ComponentPose=Local To Component(LocalPose=Use cached pose 'CtrlRig Cache'())
    Alpha:float=1
    Translation="0,0,0"
    Rotation="10,0,0"
    Scale="1,1,1"
   ) -> (Pose) {
    rotationMode=BMM_Additive
    rotationSpace=BCS_BoneSpace
    translationMode=BMM_Additive
    translationSpace=BCS_BoneSpace
    alphaBoolBlend {blendInTime=0.2, blendOption=Linear, blendOutTime=0.2}
    alphaScaleBiasClamp {bInterpResult=true, interpSpeedDecreasing=8, interpSpeedIncreasing=12}
    boneToModify {boneName=base_tie_01}
    rotation {pitch=10}
   }
   n62 = "Transform (Modify) Bone - Bone: base_tie_01"(ComponentPose=@n61, Alpha=1, Translation="0,0,0", Rotation="20,0,0", Scale="1,1,1") {
    rotationMode=BMM_Additive
    rotationSpace=BCS_BoneSpace
    translationMode=BMM_Additive
    translationSpace=BCS_BoneSpace
    alphaBoolBlend {blendInTime=0.2, blendOption=Linear, blendOutTime=0.2}
    alphaScaleBiasClamp {bInterpResult=true, interpSpeedDecreasing=8, interpSpeedIncreasing=8}
    boneToModify {boneName=base_tie_01}
    rotation {pitch=20}
   }
   n63 = RigidBody(ComponentPose=@n62) {
    overridePhysicsAsset=&/F_MED_JadeTowelGloss_Physics_Crouching
    alphaBoolBlend {blendOption=Linear}
    baseBoneRef {boneName=spine_04}
    componentAppliedLinearAccClamp=<3000 3000 3000>
    componentLinearAccScale=<0.3 0.3 0.3>
    componentLinearVelScale=<0.3 0.3 0.3>
    simSpaceSettings {worldAlpha=0.5}
   }
   n64 = "Transform (Modify) Bone - Bone: base_tie_01"!(
    ComponentPose=Local To Component(LocalPose=Use cached pose 'CtrlRig Cache'())
    bAlphaBoolEnabled:bool=True
    Translation="0,0,0"
    Rotation="21.276,-52.835,-7.312"
    Scale="1,1,1"
   ) -> (Pose) {
    alphaInputType=Bool
    rotationMode=BMM_Additive
    rotationSpace=BCS_BoneSpace
    translationMode=BMM_Additive
    translationSpace=BCS_BoneSpace
    alphaBoolBlend {blendInTime=0.1, blendOption=Linear, blendOutTime=0.2}
    alphaScaleBiasClamp {bInterpResult=true, interpSpeedDecreasing=5, interpSpeedIncreasing=5}
    boneToModify {boneName=base_tie_01}
    rotation=<rot 21.276 -52.835 -7.312>
   }
   n65 = "Transform (Modify) Bone - Bone: base_skirt_fr_r"!(ComponentPose=@n64, bAlphaBoolEnabled:bool=True, Translation="0.208,0.000,1.124", Rotation="-11.868,-6.142,1.220", Scale="1,1,1") -> (Pose) {
    alphaInputType=Bool
    rotationMode=BMM_Additive
    rotationSpace=BCS_BoneSpace
    translationMode=BMM_Additive
    translationSpace=BCS_BoneSpace
    alphaBoolBlend {blendInTime=0.1, blendOption=Linear, blendOutTime=0.2}
    alphaScaleBiasClamp {bInterpResult=true, interpSpeedDecreasing=5, interpSpeedIncreasing=5}
    boneToModify {boneName=base_skirt_fr_r}
    rotation=<rot -11.868 -6.142 1.22>
    translation {x=0.208, z=1.124}
   }
   n66 = "Transform (Modify) Bone - Bone: base_skirt_bk_r"!(ComponentPose=@n65, bAlphaBoolEnabled:bool=True, Translation="0,0,0", Rotation="6.534,44.149,-0.002", Scale="1,1,1") -> (Pose) {
    alphaInputType=Bool
    rotationMode=BMM_Additive
    rotationSpace=BCS_BoneSpace
    translationMode=BMM_Additive
    translationSpace=BCS_BoneSpace
    alphaBoolBlend {blendInTime=0.3, blendOption=Linear, blendOutTime=0.2}
    alphaScaleBiasClamp {bInterpResult=true, interpSpeedDecreasing=5, interpSpeedIncreasing=5}
    boneToModify {boneName=base_skirt_bk_r}
    rotation=<rot 6.534 44.149 -0.002>
   }
   n67 = "Transform (Modify) Bone - Bone: base_skirt_sd_r"!(ComponentPose=@n66, bAlphaBoolEnabled:bool=True, Translation="0,0,0", Rotation="-10.278,-0.000,-0.000", Scale="1,1,1") -> (Pose) {
    alphaInputType=Bool
    rotationMode=BMM_Additive
    rotationSpace=BCS_BoneSpace
    translationMode=BMM_Additive
    translationSpace=BCS_BoneSpace
    alphaBoolBlend {blendInTime=0.3, blendOption=Linear, blendOutTime=0.2}
    alphaScaleBiasClamp {bInterpResult=true, interpSpeedDecreasing=5, interpSpeedIncreasing=5}
    boneToModify {boneName=base_skirt_sd_r}
    rotation=<rot -10.278 0 0>
   }
   n68 = "Transform (Modify) Bone - Bone: base_tie_01"!(ComponentPose=@n67, Alpha:float=1, Translation="0,0,0", Rotation="0,40,0", Scale="1,1,1") -> (Pose) {
    rotationMode=BMM_Additive
    rotationSpace=BCS_BoneSpace
    translationMode=BMM_Additive
    translationSpace=BCS_BoneSpace
    alphaBoolBlend {blendInTime=0.3, blendOption=Linear, blendOutTime=0.2}
    alphaScaleBiasClamp {bInterpResult=true}
    boneToModify {boneName=base_tie_01}
    rotation {yaw=40}
   }
   n69 = "Transform (Modify) Bone - Bone: base_strap_fr_l"(ComponentPose=@n68, Alpha=1, Translation="0,0,0", Rotation="6.840,4.108,-7.927", Scale="1,1,1") {
    rotationMode=BMM_Additive
    rotationSpace=BCS_BoneSpace
    translationMode=BMM_Additive
    translationSpace=BCS_BoneSpace
    alphaBoolBlend {blendInTime=0.3, blendOption=Linear, blendOutTime=0.2}
    alphaScaleBiasClamp {bInterpResult=true, interpSpeedDecreasing=7}
    boneToModify {boneName=base_strap_fr_l}
    rotation=<rot 6.84 4.108 -7.927>
   }
   n70 = "Transform (Modify) Bone - Bone: base_strap_bk_l"(ComponentPose=@n69, Alpha=1, Translation="0,0,0", Rotation="8.962,-0.026,-3.440", Scale="1,1,1") {
    rotationMode=BMM_Additive
    rotationSpace=BCS_BoneSpace
    translationMode=BMM_Additive
    translationSpace=BCS_BoneSpace
    alphaBoolBlend {blendInTime=0.3, blendOption=Linear, blendOutTime=0.2}
    alphaScaleBiasClamp {bInterpResult=true, interpSpeedDecreasing=7, interpSpeedIncreasing=7}
    boneToModify {boneName=base_strap_bk_l}
    rotation=<rot 8.962 -0.026 -3.44>
   }
   n71 = "Transform (Modify) Bone - Bone: dyn_strap_fr_l_04"(ComponentPose=@n70, Alpha=1, Translation="0,0,2", Rotation="0,0,0", Scale="1,1,1") {
    rotationMode=BMM_Additive
    rotationSpace=BCS_BoneSpace
    translationMode=BMM_Additive
    translationSpace=BCS_BoneSpace
    alphaBoolBlend {blendInTime=0.3, blendOption=Linear, blendOutTime=0.2}
    alphaScaleBiasClamp {bInterpResult=true, interpSpeedDecreasing=7, interpSpeedIncreasing=7}
    boneToModify {boneName=dyn_strap_fr_l_04}
    translation {z=2}
   }
   n72 = RigidBody(ComponentPose=@n71) {
    bOverrideWorldGravity=true
    overridePhysicsAsset=&/F_MED_JadeTowelGloss_Physics
    simulationSpace=BaseBoneSpace
    alphaBoolBlend {blendOption=Linear}
    baseBoneRef {boneName=spine_04}
    componentAppliedLinearAccClamp=<3000 3000 3000>
    componentLinearAccScale=<0.3 0.3 0.3>
    componentLinearVelScale=<0.3 0.3 0.3>
    overrideWorldGravity {z=-980}
    simSpaceSettings {worldAlpha=0.6}
   }
   Save cached pose 'RigidBodyCache'(Pose=Blend Poses by bool(
     BlendPose_0=Component To Local(ComponentPose=@n63)
     BlendPose_1=Component To Local(ComponentPose=@n72)
     BlendTime_0=0.2
     BlendTime_1=0.4
     bActiveValue=Get Crouching Bool() -> (Crouching Bool:bool)
    ) {blendType=Linear}) {cacheName=RigidBodyCache}
  ]}
 EventGraph {nodes [
   Event OnBeginPlayingEmote() {is=title K2Node_Event}
   > Set Alpha_PlayingEmote(Alpha_PlayingEmote:double=1) -> (Output_Get:double)
   Event OnEndPlayingEmote() {is=title K2Node_Event}
   > Set Alpha_PlayingEmote(Alpha_PlayingEmote=0)
   n1 = Event BlueprintUpdateAnimation()
   > Set Alpha_DeltaTimeX(Alpha_DeltaTimeX:double=@n1.DeltaTimeX) -> (Output_Get:double)
   > n2 = Sequence!(execute) -> (
    then_0
    then_1
    then_2:Exec
    then_3:Exec
    then_4:Exec
    then_5:Exec
    then_6:Exec
    then_7:Exec
    then_8:Exec
    then_9:Exec
    then_10:Exec
    then_11:Exec
    then_12:Exec
    then_13:Exec
    then_14:Exec
    then_15:Exec
   )
   n3 = Branch(execute=@n2.then_3, Condition=Get bIsPlayingMeleeAnim())
   > Set Alpha_Melee(Alpha_Melee:double=1) -> (Output_Get:double)
   Set Alpha_Melee(execute=@n3.else, Alpha_Melee=0)
   n4 = Get bIsMoving2D()
   n5 = Branch(execute=@n2.then_5, Condition=AND Boolean(A=Get bIsJumping(), B=NOT Boolean(A=@n4)))
   > Set Alpha_Jumping Not Moving2d(Alpha_Jumping Not Moving2d:double=1) -> (Output_Get:double)
   Set Alpha_Jumping Not Moving2d(execute=@n5.else, Alpha_Jumping Not Moving2d=0)
   n6 = AND Boolean(A=Get bIsMoving2D(), B=Get bIsCrouching())
   Set Alpha_CrouchingMoving2D(
    execute=@n2.then_1
    Alpha_CrouchingMoving2D:double=FInterpTo(
     Current=Get Alpha_CrouchingMoving2D()
     Target="To Float (Boolean)"(InBool=@n6)
     DeltaTime=Get Alpha_DeltaTimeX() -> (Alpha_DeltaTimeX:double)
     InterpSpeed=10
    )
   ) -> (Output_Get:double)
   n7 = Branch(execute=@n2.then_2, Condition=Get bIsCrouching())
   > Set Alpha_Crouching(Alpha_Crouching:double=1) -> (Output_Get:double)
   Set Alpha_Crouching(execute=@n7.else, Alpha_Crouching=0)
   n8 = Get bIsMoving2D()
   n9 = Branch(execute=@n2.then_7, Condition=AND Boolean(A=NOT Boolean(A=@n8), B=Get Selected_Pickaxe() -> (Selected_Pickaxe:bool)))
   > Set Alpha_PickaxeIdle(Alpha_PickaxeIdle:double=1) -> (Output_Get:double)
   Set Alpha_PickaxeIdle(execute=@n9.else, Alpha_PickaxeIdle=0)
   n10 = Branch(execute=@n2.then_8, Condition=OR Boolean(A=Get Selected_pistol() -> (Selected_pistol:bool), B=Get Selected_Pickaxe()))
   > Set Alpha_sleeveL_FixWeapons(Alpha_sleeveL_FixWeapons:double=1) -> (Output_Get:double)
   Set Alpha_sleeveL_FixWeapons(execute=@n10.else, Alpha_sleeveL_FixWeapons=0)
   n11 = GetOwnerSkeletalMeshComponent()
   Set Test_DetectedRotAvocadoEmote(
    execute=@n2.then_13
    Test_DetectedRotAvocadoEmote:bool=Montage_IsPlaying(self=GetAnimInstance(self=@n11), Montage=/BRCosmetics/Animation/Game/MainPlayer/Emotes/AvocadoSeal/CMF/Emote_AvocadoSeal_CMF_M)
   ) -> (Output_Get:bool)
   n12 = GetOwnerSkeletalMeshComponent()
   n13 = BreakRotator(InRot=BreakTransform(InTransform=GetDeltaTransformFromRefPose(self=@n12, BoneName=neck_01, BaseName=pelvis)).Rotation)
   > n14 = Branch(Condition=AND Boolean!(A="float > float"(A=@n13, B=14), B=Get Test_DetectedRotAvocadoEmote() -> (Test_DetectedRotAvocadoEmote:bool), C:bool=Get bRigidBodyEnabled()) -> (ReturnValue))
   > Set Alpha_DetectedAvocadoEmote(Alpha_DetectedAvocadoEmote:double=1) -> (Output_Get:double)
   Set Alpha_DetectedAvocadoEmote(execute=@n14.else, Alpha_DetectedAvocadoEmote=0)
   n15 = Get bIsSkyDiving()
   n16 = Get bIsMoving2D()
   n17 = Get bIsPlayingEmote()
   n18 = Branch(
    execute=@n2.then_15
    Condition=AND Boolean!(A=NOT Boolean(A=@n16), B=Get bIsInFrontEnd(), C=NOT Boolean(A=@n17), D:bool=NOT Boolean(A=@n15)) -> (ReturnValue)
   )
   > Set FEIdle(FEIdle:double=1) -> (Output_Get:double)
   Set FEIdle(execute=@n18.else, FEIdle=0)
   n19 = Get bIsMoving2D()
   n20 = Branch(execute=@n2.then_0, Condition=AND Boolean!(A=NOT Boolean(A=@n19), B=Get bIsCrouching()) -> (ReturnValue))
   > Set Alpha_CrouchingIdle(Alpha_CrouchingIdle:double=1) -> (Output_Get:double)
   Set Alpha_CrouchingIdle(execute=@n20.else, Alpha_CrouchingIdle=0)
   n21 = GetOwnerSkeletalMeshComponent()
   n22 = BreakRotator(InRot=BreakTransform(InTransform=GetDeltaTransformFromRefPose(self=@n21, BoneName=thigh_l, BaseName=pelvis)).Rotation)
   n23 = Branch(
    execute=@n2.then_12
    Condition=AND Boolean!(A="float < float"(A=@n22.Yaw, B=-50), B=Get bIsSurfaceSwimming(), C:bool=Get bRigidBodyEnabled()) -> (ReturnValue)
   )
   > Set Alpha_UnderwaterJump(Alpha_UnderwaterJump:double=1) -> (Output_Get:double)
   Set Alpha_UnderwaterJump(execute=@n23.else, Alpha_UnderwaterJump=0)
   n24 = Get bIsKneeSliding()
   n25 = Get bIsSliding()
   n26 = Branch(execute=@n2.then_6, Condition=AND Boolean(A=NOT Boolean(A=@n24), B=NOT Boolean(A=@n25), C=Get bIsCrouching()))
   > Set Crouching Bool(Crouching Bool:bool=true) -> (Output_Get:bool)
   Delay(execute=@n26.else, Duration=0.2)
   > Set Crouching Bool(Crouching Bool=false)
   n27 = Get bIsSkyDiving()
   n28 = Get bIsMoving2D()
   n29 = Get bIsPlayingEmote()
   n30 = GetOwnerSkeletalMeshComponent()
   n31 = BreakRotator(InRot=BreakTransform(InTransform=GetDeltaTransformFromRefPose(self=@n30, BoneName=thigh_l, BaseName=pelvis)).Rotation)
   n32 = Branch(
    execute=@n2.then_14
    Condition=AND Boolean!(A="float < float"(A=@n31.Yaw, B=-30), B=NOT Boolean(A=@n28), C=NOT Boolean(A=@n29), D:bool=NOT Boolean(A=@n27), E:bool=Get bIsInFrontEnd()) -> (ReturnValue)
   )
   > Set DetectedLegBackIdleAnim(DetectedLegBackIdleAnim:double=1) -> (Output_Get:double)
   Set DetectedLegBackIdleAnim(execute=@n32.else, DetectedLegBackIdleAnim=0)
   n33 = Branch(execute=@n2.then_9, Condition=Get bIsSkyDiving())
   > Set Alpha_NotSkydiving(Alpha_NotSkydiving:double=0) -> (Output_Get:double)
   Set Alpha_NotSkydiving(execute=@n33.else, Alpha_NotSkydiving=1)
   n34 = AND Boolean!(A=NOT Boolean(A=Get bIsMoving2D()), B=NOT Boolean(A=Get bIsCrouching()), C=Get bIsOnGround() -> (bIsOnGround:bool)) -> (ReturnValue)
   n35 = Branch(execute=@n2.then_4, Condition=@n34)
   > Set Alpha_StandIdle(Alpha_StandIdle:double=1) -> (Output_Get:double)
   Set Alpha_StandIdle(execute=@n35.else, Alpha_StandIdle=0)
   n36 = Branch(execute=@n2.then_10, Condition=Get bRigidBodyEnabled())
   > Set Alpha_RB Disable(Alpha_RB Disable:double=0) -> (Output_Get:double)
   Set Alpha_RB Disable(execute=@n36.else, Alpha_RB Disable=1)
   n37 = Branch(execute=@n2.then_11, Condition=Get bIsDBNO())
   > Set Alpha_DBNO(Alpha_DBNO:double=1) -> (Output_Get:double)
   Set Alpha_DBNO(execute=@n37.else, Alpha_DBNO=0)
   Event OnWeaponChanged()
   > n38 = Cast To FortPlayerAnimInstance(Object=GetAnimInstance(self=GetOwnerSkeletalMeshComponent()))
   > Set Selected_Pickaxe(Selected_Pickaxe:bool=false) -> (Output_Get:bool)
   > Set Selected_pistol(Selected_pistol:bool=false) -> (Output_Get:bool)
   > Set Selected_Shotgun(Selected_Shotgun:bool=false) -> (Output_Get:bool)
   > n39 = Switch on EFortWeaponCoreAnimation!(execute, Selection:EFortWeaponCoreAnimation Enum=Get WeaponCoreAnim(self=@n38)) -> (
    Melee:Exec
    Pistol:Exec
    Shotgun:Exec
    PaperBlueprint:Exec
    Rifle:Exec
    MeleeOneHand:Exec
    MachinePistol:Exec
    RocketLauncher:Exec
    GrenadeLauncher:Exec
    GoingCommando:Exec
    AssaultRifle:Exec
    TacticalShotgun:Exec
    SniperRifle:Exec
    TrapPlacement:Exec
    ShoulderLauncher:Exec
    AbilityDecoTool:Exec
    Crossbow:Exec
    C4:Exec
    RemoteControl:Exec
    DualWield:Exec
    AR_BullPup:Exec
    AR_ForwardGrip:Exec
    MedPackPaddles:Exec
    SMG_P90:Exec
    AR_DrumGun:Exec
    Consumable_Small:Exec
    Consumable_Large:Exec
    Balloon:Exec
    MountedTurret:Exec
    CreativeTool:Exec
    ExplosiveBow:Exec
    AshtonIndigo:Exec
    AshtonChicago:Exec
    MeleeDualWield:Exec
    Unarmed:Exec
   )
   > Set Selected_Pickaxe(Selected_Pickaxe=1)
   Set Selected_pistol(execute=@n39.Pistol, Selected_pistol=1)
   Set Selected_Shotgun(execute=@n39.Shotgun, Selected_Shotgun=true)
   OR Boolean(A=@n34, B=NOT Boolean(A=Get bRigidBodyEnabled()))
  ]}
]
variables [
 Alpha_Crouching {type=number}
 Alpha_CrouchingIdle {type=number}
 Alpha_CrouchingMoving2D {type=number}
 Alpha_DBNO {type=number}
 Alpha_DeltaTimeX {type=number}
 Alpha_DetectedAvocadoEmote {type=number}
 Alpha_Dynamics {type=number}
 Alpha_Jumping Not Moving2d {type=number}
 Alpha_Melee {type=number}
 Alpha_MeleeCrouching {type=number}
 Alpha_NotSkydiving {type=number}
 Alpha_PickaxeIdle {type=number}
 Alpha_PlayingEmote {type=number}
 Alpha_RB Disable {type=number}
 Alpha_sleeveL_FixWeapons {type=number}
 Alpha_StandIdle {type=number}
 Alpha_UnderwaterJump {type=number}
 Backward {type=number}
 Crouching Bool {type=boolean}
 Detected_RotateHandWithWeapons {type=number}
 Detected_RotateHandWithWeaponsLow {type=number}
 DetectedLegBackIdleAnim {type=number}
 FEIdle {type=number}
 Forward {type=number}
 Left {type=number}
 Right {type=number}
 Right_leg_rot {type=boolean}
 Selected_Pickaxe {type=boolean}
 Selected_pistol {type=boolean}
 Selected_Shotgun {type=boolean}
 Test_DetectedRotAvocadoEmote {type=boolean}
]
