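# Physics asset for the front and side hair of the F_MED_JadeTowelGloss face accessory, giving the female character's dynamic hair bones capsule bodies.
# Short capsule chains and single strands hang from one large head capsule, with the right side A root branching into two children.
# Every constraint locks twist and allows only narrow, per-joint swing, often tighter at the roots, so the hair sways while keeping its shape.
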PhysicsAsset /BRCosmetics/Characters/Player/Female/Medium/Bodies/F_MED_JadeTowelGloss/Meshes/Parts/F_MED_JadeTowelGloss_FaceAcc_PhysicsHairFront
bodies {
 dyn_hair_fr_a_01_c [{shapeType=Capsule, length=1.5, radius=1.5, shapeName=dyn_hair_fr_a_01_c_capsule, rotation=<rot 90 0 0>}]
 dyn_hair_fr_a_02_c [{shapeType=Capsule, length=1.5, radius=1.5, shapeName=dyn_hair_fr_a_02_c_capsule, rotation=<rot 90 0 0>}]
 dyn_hair_fr_a_03_c [{
   shapeType=Capsule
   length=3.5
   radius=1.5
   shapeName=dyn_hair_fr_a_03_c_capsule
   center=<1.2522 0.0001 0.3636>
   rotation=<rot 73.8094 -179.9875 -179.9937>
  }]
 dyn_hair_fr_b_l [{shapeType=Capsule, length=2.5, radius=1, shapeName=dyn_hair_fr_b_l_capsule, center=<1.0196 0 0>, rotation=<rot 90 0 0>}]
 dyn_hair_sd_a_01_l [{shapeType=Capsule, length=3, radius=2, shapeName=dyn_hair_sd_a_01_l_capsule, center=<0.8447 0 0>, rotation=<rot 90 0 0>}]
 dyn_hair_sd_a_01_r [{
   shapeType=Capsule
   length=4.2861
   radius=1.47
   shapeName=dyn_hair_sd_a_01_r_capsule
   center=<2.4902 0.4278 -0.1335>
   rotation=<rot 87.1583 -165.7673 -179.9995>
  }]
 dyn_hair_sd_a_02_l [{shapeType=Capsule, length=2.5, radius=2, shapeName=dyn_hair_sd_a_02_l_capsule, center=<1.0901 0 0>, rotation=<rot 86.0708 -180 -180>}]
 dyn_hair_sd_a_02_r [{shapeType=Capsule, length=2.8071, radius=1.8959, shapeName=dyn_hair_sd_a_02_r_capsule, center=<1.5253 0 -0.1948>, rotation=<rot 90 0 0>}]
 dyn_hair_sd_a_r [{
   shapeType=Capsule
   length=4.5943
   radius=2.2382
   shapeName=dyn_hair_sd_a_r_capsule
   center=<0 0.3997 -1.4059>
   rotation=<rot 0 0.0001 74.1304>
  }]
 dyn_hair_sd_b_01_l [{shapeType=Capsule, length=2, radius=2, shapeName=dyn_hair_sd_b_01_l_capsule, center=<0 0 0.9529>, rotation=<rot 0 0 90>}]
 dyn_hair_sd_b_01_r [{shapeType=Capsule, length=2, radius=1.5, shapeName=dyn_hair_sd_b_01_r_capsule, rotation=<rot 90 0 0>}]
 dyn_hair_sd_b_02_l [{shapeType=Capsule, length=2.5, radius=1.5, shapeName=dyn_hair_sd_b_02_l_capsule, center=<0.4375 0 0.469>, rotation=<rot 0 0 90>}]
 dyn_hair_sd_b_02_r [{shapeType=Capsule, length=2, radius=1, shapeName=dyn_hair_sd_b_02_r_capsule, rotation=<rot 90 0 0>}]
 dyn_hair_sd_b_03_l [{shapeType=Capsule, length=2, radius=1.5, shapeName=dyn_hair_sd_b_03_l_capsule, center=<0.6401 0 0.2529>, rotation=<rot 90 0 0>}]
 dyn_hair_sd_b_03_r [{
   shapeType=Capsule
   length=3
   radius=1
   shapeName=dyn_hair_sd_b_03_r_capsule
   center=<1.3586 0 -0.0874>
   rotation=<rot 82.6427 179.9999 179.9999>
  }]
 dyn_hair_sd_c_01_l [{
   shapeType=Capsule
   length=3
   radius=1
   shapeName=dyn_hair_sd_c_01_l_capsule
   center=<2.0755 -0.2742 0.0057>
   rotation=<rot 84.7144 -6.9147 0.0001>
  }]
 head [{shapeType=Capsule, length=4.5, radius=8, shapeName=head_capsule, center=<3.2209 1.3624 0>, rotation=<rot 90 0 0>}]
}
constraints [
 {
  bone1Name=dyn_hair_fr_a_03_c
  bone2Name=dyn_hair_fr_a_02_c
  swing1LimitDegrees=20
  swing2LimitDegrees=30
  twistLimitDegrees=45
  twistMotion=Locked
 }
 {
  bone1Name=dyn_hair_fr_a_02_c
  bone2Name=dyn_hair_fr_a_01_c
  swing1LimitDegrees=20
  swing2LimitDegrees=25
  twistLimitDegrees=45
  twistMotion=Locked
 }
 {bone1Name=dyn_hair_fr_a_01_c, bone2Name=head, swing1LimitDegrees=10, swing2LimitDegrees=25, twistLimitDegrees=45, twistMotion=Locked}
 {
  bone1Name=dyn_hair_sd_b_03_l
  bone2Name=dyn_hair_sd_b_02_l
  swing1LimitDegrees=30
  swing2LimitDegrees=15
  twistLimitDegrees=45
  twistMotion=Locked
 }
 {
  bone1Name=dyn_hair_sd_b_02_l
  bone2Name=dyn_hair_sd_b_01_l
  swing1LimitDegrees=15
  swing2LimitDegrees=15
  twistLimitDegrees=45
  twistMotion=Locked
 }
 {bone1Name=dyn_hair_sd_b_01_l, bone2Name=head, swing1LimitDegrees=7, swing2LimitDegrees=15, twistLimitDegrees=45, twistMotion=Locked}
 {
  bone1Name=dyn_hair_sd_a_02_r
  bone2Name=dyn_hair_sd_a_r
  swing1LimitDegrees=20
  swing2LimitDegrees=10
  twistLimitDegrees=45
  twistMotion=Locked
 }
 {
  bone1Name=dyn_hair_sd_a_01_r
  bone2Name=dyn_hair_sd_a_r
  swing1LimitDegrees=20
  swing2LimitDegrees=10
  twistLimitDegrees=45
  twistMotion=Locked
 }
 {bone1Name=dyn_hair_sd_a_r, bone2Name=head, swing1LimitDegrees=20, swing2LimitDegrees=10, twistLimitDegrees=45, twistMotion=Locked}
 {bone1Name=dyn_hair_fr_b_l, bone2Name=head, swing1LimitDegrees=20, swing2LimitDegrees=20, twistLimitDegrees=45, twistMotion=Locked}
 {
  bone1Name=dyn_hair_sd_b_03_r
  bone2Name=dyn_hair_sd_b_02_r
  swing1LimitDegrees=15
  swing2LimitDegrees=20
  twistLimitDegrees=45
  twistMotion=Locked
 }
 {
  bone1Name=dyn_hair_sd_b_02_r
  bone2Name=dyn_hair_sd_b_01_r
  swing1LimitDegrees=15
  swing2LimitDegrees=20
  twistLimitDegrees=45
  twistMotion=Locked
 }
 {bone1Name=dyn_hair_sd_b_01_r, bone2Name=head, swing1LimitDegrees=15, swing2LimitDegrees=10, twistLimitDegrees=45, twistMotion=Locked}
 {bone1Name=dyn_hair_sd_c_01_l, bone2Name=head, swing1LimitDegrees=20, swing2LimitDegrees=20, twistLimitDegrees=45, twistMotion=Locked}
 {bone1Name=dyn_hair_sd_a_01_l, bone2Name=head, swing1LimitDegrees=5, swing2LimitDegrees=5, twistLimitDegrees=45, twistMotion=Locked}
 {
  bone1Name=dyn_hair_sd_a_02_l
  bone2Name=dyn_hair_sd_a_01_l
  swing1LimitDegrees=12
  swing2LimitDegrees=15
  twistLimitDegrees=45
  twistMotion=Locked
 }
]
